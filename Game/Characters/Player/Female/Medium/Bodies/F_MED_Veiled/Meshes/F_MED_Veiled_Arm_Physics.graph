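# Physics asset for the arm mesh of the medium female Veiled player body, simulating a dangling loop on the right arm.
# Capsule bodies form two bone chains, front and back, rooted at master bodies linked to upperarm_r, alongside capsules for the upper and lower arm.
# Angular limits shape each chain, while a free-rotation constraint joins the two tips to close the loop.
# The dyn_arm_loop_fr1 body has no constraint.
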PhysicsAsset /Game/Characters/Player/Female/Medium/Bodies/F_MED_Veiled/Meshes/F_MED_Veiled_Arm_Physics
bodies {
 dyn_arm_loop_01 [{
   shapeType=Capsule
   length=6.3752
   radius=1.4652
   shapeName=dyn_arm_loop_01_capsule
   center=<9.3281 -5.4068 0.0989>
   rotation=<rot 0 90.4274 -85.615>
  }]
 dyn_arm_loop_02 [{shapeType=Capsule, length=6.755, radius=3.88, shapeName=dyn_arm_loop_02_capsule, rotation=<rot 0 0 90>}]
 dyn_arm_loop_03 [{shapeType=Capsule, length=4.9475, radius=3.614, shapeName=dyn_arm_loop_03_capsule, center=<4 0 1.5>, rotation=<rot 0 0 100.0001>}]
 dyn_arm_loop_MSTR [{shapeType=Capsule, length=0.505, radius=0.505, shapeName=dyn_arm_loop_MSTR_capsule, rotation=<rot 0 -90 0>}]
 dyn_arm_loop_bk [{shapeType=Capsule, length=3.2545, radius=2.6969, shapeName=dyn_arm_loop_bk_capsule, center=<5 -1 0>, rotation=<rot 90 0 0>}]
 dyn_arm_loop_bk_02 [{shapeType=Capsule, length=2.3444, radius=2.8263, shapeName=dyn_arm_loop_bk_02_capsule, center=<4.5 0 0>, rotation=<rot 90 0 0>}]
 dyn_arm_loop_bk_03 [{
   shapeType=Capsule
   length=1.4198
   radius=4.2169
   shapeName=dyn_arm_loop_bk_03_capsule
   center=<3.5284 3 -1>
   rotation=<rot 0 -48.706 -76.3187>
  }]
 dyn_arm_loop_bk_MSTR [{shapeType=Capsule, length=0.505, radius=0.505, shapeName=dyn_arm_loop_bk_MSTR_capsule, rotation=<rot 0 -90 0>}]
 dyn_arm_loop_fr1 [{shapeType=Capsule, length=0.505, radius=0.505, shapeName=dyn_arm_loop_fr1_capsule, rotation=<rot 0 -90 0>}]
 lowerarm_r [{
   shapeType=Capsule
   length=5.25
   radius=15.13
   shapeName=lowerarm_r_capsule
   center=<-12.8941 0.7398 11.4799>
   rotation=<rot 85 -179.9995 -179.9997>
  }]
 upperarm_r [{
   shapeType=Capsule
   length=14.0657
   radius=5.4158
   shapeName=upperarm_r_capsule
   center=<-16.8159 0.6052 2.7659>
   rotation=<rot -29.4371 -84.302 -100.0598>
  }]
}
constraints [
 {
  bone1Name=dyn_arm_loop_MSTR
  bone2Name=upperarm_r
  swing1LimitDegrees=45
  swing1Motion=Free
  swing2LimitDegrees=45
  swing2Motion=Free
  twistLimitDegrees=45
  twistMotion=Free
 }
 {
  bone1Name=dyn_arm_loop_bk_MSTR
  bone2Name=dyn_arm_loop_MSTR
  swing1LimitDegrees=45
  swing1Motion=Free
  swing2LimitDegrees=45
  swing2Motion=Free
  twistLimitDegrees=45
  twistMotion=Free
 }
 {
  bone1Name=lowerarm_r
  bone2Name=upperarm_r
  swing1LimitDegrees=45
  swing1Motion=Free
  swing2LimitDegrees=45
  swing2Motion=Free
  twistLimitDegrees=45
  twistMotion=Free
 }
 {bone1Name=dyn_arm_loop_bk, bone2Name=dyn_arm_loop_bk_MSTR, swing1LimitDegrees=5, swing2LimitDegrees=5, twistLimitDegrees=5}
 {bone1Name=dyn_arm_loop_bk_03, bone2Name=dyn_arm_loop_bk_02, swing1LimitDegrees=45, swing2LimitDegrees=45, twistLimitDegrees=5}
 {bone1Name=dyn_arm_loop_bk_02, bone2Name=dyn_arm_loop_bk, swing1LimitDegrees=45, swing2LimitDegrees=45, twistLimitDegrees=5}
 {bone1Name=dyn_arm_loop_03, bone2Name=dyn_arm_loop_02, swing1LimitDegrees=45, swing2LimitDegrees=45, twistLimitDegrees=25}
 {
  bone1Name=dyn_arm_loop_01
  bone2Name=dyn_arm_loop_MSTR
  swing1LimitDegrees=45
  swing1Motion=Free
  swing2LimitDegrees=45
  swing2Motion=Free
  twistLimitDegrees=45
  twistMotion=Free
 }
 {bone1Name=dyn_arm_loop_02, bone2Name=dyn_arm_loop_01, swing1LimitDegrees=45, swing2LimitDegrees=45, twistLimitDegrees=15}
 {
  bone1Name=dyn_arm_loop_bk_03
  bone2Name=dyn_arm_loop_03
  swing1LimitDegrees=45
  swing1Motion=Free
  swing2LimitDegrees=45
  swing2Motion=Free
  twistLimitDegrees=45
  twistMotion=Free
 }
]
