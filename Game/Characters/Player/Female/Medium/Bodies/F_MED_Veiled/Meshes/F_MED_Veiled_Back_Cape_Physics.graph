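# Physics asset for the back cape mesh of the Veiled medium female player body, covering dangling garment bones.
# Pelvis and leg shapes plus a large FloorCOL sphere provide body and floor collision volumes.
# Tiny driver/MSTR anchor capsules link sweater box panels, two back-strap chains and six flaps per calf to pelvis or legs.
# Most hanging pieces have limited swing and twist, while some neighboring sweater panels are cross-linked with free angular constraints.
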PhysicsAsset /Game/Characters/Player/Female/Medium/Bodies/F_MED_Veiled/Meshes/F_MED_Veiled_Back_Cape_Physics
bodies {
 FloorCOL [{shapeType=Sphere, radius=183.6686, shapeName=floorCOL_sphere, center=<-3.4923 -2.7141 -174.8455>}]
 calf_l [{shapeType=Capsule, length=31.75, radius=11.75, shapeName=calf_l_capsule, center=<-27 0 0>, rotation=<rot 90 80.7734 75.7735>}]
 calf_r [{shapeType=Capsule, length=31.75, radius=11.75, shapeName=calf_r_capsule, center=<27 0 0>, rotation=<rot 90 80.7735 75.7736>}]
 calf_twist_02_l [{shapeType=Capsule, length=28.75, radius=6.25, shapeName=calf_twist_02_r_capsule, center=<-5.6018 0 0>, rotation=<rot 90 0 0>}]
 calf_twist_02_r [{shapeType=Capsule, length=27.875, radius=7.125, shapeName=calf_twist_02_r_capsule, center=<4.158 1.1754 0>, rotation=<rot 85 -175.0002 180>}]
 driver_sweater_bck [{shapeType=Capsule, length=0.505, radius=0.505, shapeName=driver_sweater_bck_capsule, rotation=<rot 0 -90 0>}]
 driver_sweater_bck_l [{shapeType=Capsule, length=0.505, radius=0.505, shapeName=driver_sweater_bck_l_capsule, rotation=<rot 0 -90 0>}]
 driver_sweater_bck_r [{shapeType=Capsule, length=0.505, radius=0.505, shapeName=driver_sweater_bck_r_capsule, rotation=<rot 0 -90 0>}]
 driver_sweater_frt_l [{shapeType=Capsule, length=0.505, radius=0.505, shapeName=driver_sweater_frt_l_capsule, rotation=<rot 0 -90 0>}]
 driver_sweater_frt_r [{shapeType=Capsule, length=0.505, radius=0.505, shapeName=driver_sweater_frt_r_capsule, rotation=<rot 0 -90 0>}]
 driver_sweater_side_l [{shapeType=Capsule, length=0.505, radius=0.505, shapeName=driver_sweater_side_l_capsule, rotation=<rot 0 -90 0>}]
 driver_sweater_side_r [{shapeType=Capsule, length=0.505, radius=0.505, shapeName=driver_sweater_side_r_capsule, rotation=<rot 0 -90 0>}]
 dyn_back_strap_l_01 [{shapeType=Capsule, length=3.5, radius=1.5, shapeName=dyn_back_strap_l_01_capsule, center=<2 -1 0>, rotation=<rot 90 0 0>}]
 dyn_back_strap_l_02 [{shapeType=Capsule, length=9.75, radius=2.5, shapeName=dyn_back_strap_l_02_capsule, center=<8 0 0>, rotation=<rot 90 0 0>}]
 dyn_back_strap_l_03 [{shapeType=Capsule, length=15, radius=2, shapeName=dyn_back_strap_l_03_capsule, center=<10 0 0>, rotation=<rot 90 0 0>}]
 dyn_back_strap_l_04 [{shapeType=Capsule, length=15, radius=2, shapeName=dyn_back_strap_l_04_capsule, center=<12 0 0>, rotation=<rot 90 0 0>}]
 dyn_back_strap_l_MSTR [{shapeType=Capsule, length=0.505, radius=0.505, shapeName=dyn_back_strap_l_MSTR_capsule, rotation=<rot 0 -90 0>}]
 dyn_back_strap_r_01 [{shapeType=Capsule, length=1.75, radius=2.375, shapeName=dyn_back_strap_r_01_capsule, center=<2 0 0>, rotation=<rot 90 0 0>}]
 dyn_back_strap_r_02 [{shapeType=Capsule, length=9.625, radius=2.625, shapeName=dyn_back_strap_r_02_capsule, center=<9 0 0>, rotation=<rot 90 0 0>}]
 dyn_back_strap_r_03 [{shapeType=Capsule, length=12, radius=2, shapeName=dyn_back_strap_r_03_capsule, center=<10 0 0>, rotation=<rot 90 0 0>}]
 dyn_back_strap_r_04 [{shapeType=Capsule, length=12, radius=2, shapeName=dyn_back_strap_r_04_capsule, center=<8 0 0>, rotation=<rot 90 0 0>}]
 dyn_back_strap_r_05 [{shapeType=Capsule, length=20, radius=2, shapeName=dyn_back_strap_r_05_capsule, center=<12 0 0>, rotation=<rot 90 0 0>}]
 dyn_back_strap_r_MSTR [{shapeType=Capsule, length=0.505, radius=0.505, shapeName=dyn_back_strap_r_MSTR_capsule, rotation=<rot 0 -90 0>}]
 dyn_calf_l_01 [{shapeType=Capsule, length=5.125, radius=3.875, shapeName=dyn_calf_l_01_capsule, center=<3 0 3>, rotation=<rot 90 0 0>}]
 dyn_calf_l_02 [{shapeType=Capsule, length=4.125, radius=4.875, shapeName=dyn_calf_l_02_capsule, center=<3 0 5>, rotation=<rot 90 0 0>}]
 dyn_calf_l_03 [{shapeType=Capsule, length=4.625, radius=4.375, shapeName=dyn_calf_l_03_capsule, center=<3 0 2>, rotation=<rot 90 0 0>}]
 dyn_calf_l_04 [{shapeType=Capsule, length=4.5, radius=4.5, shapeName=dyn_calf_l_04_capsule, center=<3 0 -2>, rotation=<rot 90 0 0>}]
 dyn_calf_l_05 [{shapeType=Capsule, length=5, radius=4, shapeName=dyn_calf_l_05_capsule, center=<3 0 0>, rotation=<rot 90 0 0>}]
 dyn_calf_l_06 [{shapeType=Capsule, length=4.625, radius=4.375, shapeName=dyn_calf_l_06_capsule, center=<3 0 2>, rotation=<rot 90 0 0>}]
 dyn_calf_l_MSTR_01 [{shapeType=Capsule, length=0.505, radius=0.505, shapeName=dyn_calf_l_MSTR_01_capsule, rotation=<rot 0 -90 0>}]
 dyn_calf_l_MSTR_02 [{shapeType=Capsule, length=0.505, radius=0.505, shapeName=dyn_calf_l_MSTR_02_capsule, rotation=<rot 0 -90 0>}]
 dyn_calf_l_MSTR_03 [{shapeType=Capsule, length=0.505, radius=0.505, shapeName=dyn_calf_l_MSTR_03_capsule, rotation=<rot 0 -90 0>}]
 dyn_calf_l_MSTR_04 [{shapeType=Capsule, length=0.505, radius=0.505, shapeName=dyn_calf_l_MSTR_04_capsule, rotation=<rot 0 -90 0>}]
 dyn_calf_l_MSTR_05 [{shapeType=Capsule, length=0.505, radius=0.505, shapeName=dyn_calf_l_MSTR_05_capsule, rotation=<rot 0 -90 0>}]
 dyn_calf_l_MSTR_06 [{shapeType=Capsule, length=0.505, radius=0.505, shapeName=dyn_calf_l_MSTR_06_capsule, rotation=<rot 0 -90 0>}]
 dyn_calf_r_01 [{shapeType=Capsule, length=4.75, radius=4.25, shapeName=dyn_calf_r_01_capsule, center=<3 0 0>, rotation=<rot 90 0 0>}]
 dyn_calf_r_02 [{shapeType=Capsule, length=4.625, radius=4.375, shapeName=dyn_calf_r_02_capsule, center=<3 0 1>, rotation=<rot 90 0 0>}]
 dyn_calf_r_03 [{shapeType=Capsule, length=4, radius=5, shapeName=dyn_calf_r_03_capsule, center=<3 0 2>, rotation=<rot 90 0 0>}]
 dyn_calf_r_04 [{shapeType=Capsule, length=5, radius=4, shapeName=dyn_calf_r_04_capsule, center=<3 0 2>, rotation=<rot 90 0 0>}]
 dyn_calf_r_05 [{shapeType=Capsule, length=4, radius=5, shapeName=dyn_calf_r_05_capsule, center=<3 0 5>, rotation=<rot 90 0 0>}]
 dyn_calf_r_06 [{shapeType=Capsule, length=4.75, radius=4.25, shapeName=dyn_calf_r_06_capsule, center=<3 0 2>, rotation=<rot 90 0 0>}]
 dyn_calf_r_MSTR_01 [{shapeType=Capsule, length=0.505, radius=0.505, shapeName=dyn_calf_r_MSTR_01_capsule, rotation=<rot 0 -90 0>}]
 dyn_calf_r_MSTR_02 [{shapeType=Capsule, length=0.505, radius=0.505, shapeName=dyn_calf_r_MSTR_02_capsule, rotation=<rot 0 -90 0>}]
 dyn_calf_r_MSTR_03 [{shapeType=Capsule, length=0.505, radius=0.505, shapeName=dyn_calf_r_MSTR_03_capsule, rotation=<rot 0 -90 0>}]
 dyn_calf_r_MSTR_04 [{shapeType=Capsule, length=0.505, radius=0.505, shapeName=dyn_calf_r_MSTR_04_capsule, rotation=<rot 0 -90 0>}]
 dyn_calf_r_MSTR_05 [{shapeType=Capsule, length=0.505, radius=0.505, shapeName=dyn_calf_r_MSTR_05_capsule, rotation=<rot 0 -90 0>}]
 dyn_calf_r_MSTR_06 [{shapeType=Capsule, length=0.505, radius=0.505, shapeName=dyn_calf_r_MSTR_06_capsule, rotation=<rot 0 -90 0>}]
 dyn_sweater_Driven_Bone [{shapeType=Capsule, length=0.505, radius=0.505, shapeName=dyn_sweater_Driven_Bone_capsule, rotation=<rot 0 -90 0>}]
 dyn_sweater_bck [{shapeType=Box, extentX=15, extentY=19, extentZ=7.5, center=<9 0 0>}]
 dyn_sweater_bck_l [{shapeType=Box, extentX=19, extentY=14.5, extentZ=4.5, center=<9 0 0>, rotation=<rot 3.3801 4.1386 34.3148>}]
 dyn_sweater_bck_r [{shapeType=Box, extentX=19, extentY=14.5, extentZ=5, center=<9.2891 0 0>, rotation=<rot 0 0 -30>}]
 dyn_sweater_frt [{shapeType=Box, extentX=13.5, extentY=4, extentZ=20.5, center=<0 0.3819 -9.1199>, rotation=<rot 0 0 -3.1066>}]
 dyn_sweater_frt_l [{shapeType=Box, extentX=19.5, extentY=14, extentZ=4.5, center=<9 0 1>, rotation=<rot 5.2582 -2.4497 -18.258>}]
 dyn_sweater_frt_r [{shapeType=Box, extentX=19.5, extentY=15, extentZ=4.5, center=<9.5469 0 1>, rotation=<rot 0.0001 0 20>}]
 dyn_sweater_side_l [{shapeType=Box, extentX=18, extentY=14.5, extentZ=4, center=<9 0 3>, rotation=<rot 3.6829 -0.3573 3.5002>}]
 dyn_sweater_side_r [{shapeType=Box, extentX=18, extentY=14.5, extentZ=4, center=<9.2283 0 3>, rotation=<rot 2.8621 -0.4075 -176.5024>}]
 pelvis [{shapeType=Capsule, length=10.5603, radius=10.1851, center=<-8.2175 -1.9871 0>}]
 thigh_l [
  {shapeType=Capsule, length=19.5532, radius=6.9331, center=<-21.2717 -0.3685 1.6327>, rotation=<rot 90 89.4327 89.433>}
  {shapeType=Capsule, length=5.375, radius=6.75, shapeName=thigh_r_capsule, center=<-39.4665 0.7099 1.7776>, rotation=<rot 90 0 0>}
 ]
 thigh_r [
  {shapeType=Capsule, length=19.5532, radius=6.9331, center=<21.2717 -0.3685 -1.6327>, rotation=<rot 90 89.4327 89.433>}
  {shapeType=Capsule, length=5.375, radius=6.75, shapeName=thigh_r_capsule, center=<39.4665 0.7099 -1.7776>, rotation=<rot 90 0 0>}
 ]
 thigh_twist_01_l [{shapeType=Capsule, length=23.5, radius=11, shapeName=thigh_twist_01_l_capsule, center=<8 0 0>, rotation=<rot 90 0 0>}]
 thigh_twist_01_r [{shapeType=Capsule, length=23.5, radius=11, shapeName=thigh_twist_01_r_capsule, center=<-8 0 0>, rotation=<rot 90 0 0>}]
}
constraints [
 {
  bone1Name=thigh_r
  bone2Name=pelvis
  swing1LimitDegrees=45
  swing1Motion=Free
  swing2LimitDegrees=45
  swing2Motion=Free
  twistLimitDegrees=45
  twistMotion=Free
 }
 {bone1Name=thigh_l, bone2Name=pelvis, swing1LimitDegrees=45, swing2LimitDegrees=45, twistLimitDegrees=45}
 {
  bone1Name=calf_twist_02_l
  bone2Name=calf_l
  swing1LimitDegrees=45
  swing1Motion=Free
  swing2LimitDegrees=45
  swing2Motion=Free
  twistLimitDegrees=45
  twistMotion=Free
 }
 {
  bone1Name=calf_twist_02_r
  bone2Name=calf_r
  swing1LimitDegrees=45
  swing1Motion=Free
  swing2LimitDegrees=45
  swing2Motion=Free
  twistLimitDegrees=45
  twistMotion=Free
 }
 {bone1Name=dyn_sweater_side_r, bone2Name=driver_sweater_side_r, swing1LimitDegrees=45, swing2LimitDegrees=45, twistLimitDegrees=15}
 {
  bone1Name=driver_sweater_bck
  bone2Name=pelvis
  swing1LimitDegrees=45
  swing1Motion=Free
  swing2LimitDegrees=45
  swing2Motion=Free
  twistLimitDegrees=45
  twistMotion=Free
 }
 {
  bone1Name=dyn_sweater_bck
  bone2Name=driver_sweater_bck
  swing1LimitDegrees=45
  swing1Motion=Free
  swing2LimitDegrees=45
  swing2Motion=Free
  twistLimitDegrees=45
  twistMotion=Free
 }
 {
  bone1Name=dyn_sweater_Driven_Bone
  bone2Name=pelvis
  swing1LimitDegrees=45
  swing1Motion=Free
  swing2LimitDegrees=45
  swing2Motion=Free
  twistLimitDegrees=45
  twistMotion=Free
 }
 {
  bone1Name=dyn_sweater_frt
  bone2Name=dyn_sweater_Driven_Bone
  swing1LimitDegrees=45
  swing1Motion=Free
  swing2LimitDegrees=45
  swing2Motion=Free
  twistLimitDegrees=45
  twistMotion=Free
 }
 {
  bone1Name=driver_sweater_bck_l
  bone2Name=thigh_l
  swing1LimitDegrees=45
  swing1Motion=Free
  swing2LimitDegrees=45
  swing2Motion=Free
  twistLimitDegrees=45
  twistMotion=Free
 }
 {bone1Name=dyn_sweater_bck_l, bone2Name=driver_sweater_bck_l, swing1LimitDegrees=45, swing2LimitDegrees=45, twistLimitDegrees=15}
 {
  bone1Name=driver_sweater_frt_l
  bone2Name=thigh_l
  swing1LimitDegrees=45
  swing1Motion=Free
  swing2LimitDegrees=45
  swing2Motion=Free
  twistLimitDegrees=45
  twistMotion=Free
 }
 {bone1Name=dyn_sweater_frt_l, bone2Name=driver_sweater_frt_l, swing1LimitDegrees=45, swing2LimitDegrees=45, twistLimitDegrees=15}
 {
  bone1Name=driver_sweater_side_l
  bone2Name=thigh_l
  swing1LimitDegrees=45
  swing1Motion=Free
  swing2LimitDegrees=45
  swing2Motion=Free
  twistLimitDegrees=45
  twistMotion=Free
 }
 {bone1Name=dyn_sweater_side_l, bone2Name=driver_sweater_side_l, swing1LimitDegrees=45, swing2LimitDegrees=45, twistLimitDegrees=15}
 {
  bone1Name=driver_sweater_bck_r
  bone2Name=thigh_r
  swing1LimitDegrees=45
  swing1Motion=Free
  swing2LimitDegrees=45
  swing2Motion=Free
  twistLimitDegrees=45
  twistMotion=Free
 }
 {bone1Name=dyn_sweater_bck_r, bone2Name=driver_sweater_bck_r, swing1LimitDegrees=45, swing2LimitDegrees=45, twistLimitDegrees=15}
 {
  bone1Name=driver_sweater_frt_r
  bone2Name=thigh_r
  swing1LimitDegrees=45
  swing1Motion=Free
  swing2LimitDegrees=45
  swing2Motion=Free
  twistLimitDegrees=45
  twistMotion=Free
 }
 {bone1Name=dyn_sweater_frt_r, bone2Name=driver_sweater_frt_r, swing1LimitDegrees=45, swing2LimitDegrees=45, twistLimitDegrees=15}
 {
  bone1Name=driver_sweater_side_r
  bone2Name=thigh_r
  swing1LimitDegrees=45
  swing1Motion=Free
  swing2LimitDegrees=45
  swing2Motion=Free
  twistLimitDegrees=45
  twistMotion=Free
 }
 {
  bone1Name=dyn_sweater_bck
  bone2Name=dyn_sweater_bck_l
  swing1LimitDegrees=45
  swing1Motion=Free
  swing2LimitDegrees=45
  swing2Motion=Free
  twistLimitDegrees=45
  twistMotion=Free
 }
 {
  bone1Name=dyn_sweater_bck
  bone2Name=dyn_sweater_bck_r
  swing1LimitDegrees=45
  swing1Motion=Free
  swing2LimitDegrees=45
  swing2Motion=Free
  twistLimitDegrees=45
  twistMotion=Free
 }
 {
  bone1Name=dyn_sweater_side_l
  bone2Name=dyn_sweater_frt_l
  swing1LimitDegrees=45
  swing1Motion=Free
  swing2LimitDegrees=45
  swing2Motion=Free
  twistLimitDegrees=45
  twistMotion=Free
 }
 {
  bone1Name=dyn_sweater_side_r
  bone2Name=dyn_sweater_frt_r
  swing1LimitDegrees=45
  swing1Motion=Free
  swing2LimitDegrees=45
  swing2Motion=Free
  twistLimitDegrees=45
  twistMotion=Free
 }
 {
  bone1Name=dyn_back_strap_l_MSTR
  bone2Name=pelvis
  swing1LimitDegrees=45
  swing1Motion=Free
  swing2LimitDegrees=45
  swing2Motion=Free
  twistLimitDegrees=45
  twistMotion=Free
 }
 {bone1Name=dyn_back_strap_l_01, bone2Name=dyn_back_strap_l_MSTR, swing1LimitDegrees=35, swing2LimitDegrees=5, twistLimitDegrees=3}
 {bone1Name=dyn_back_strap_l_02, bone2Name=dyn_back_strap_l_01, swing1LimitDegrees=35, swing2LimitDegrees=20, twistLimitDegrees=15}
 {bone1Name=dyn_back_strap_l_03, bone2Name=dyn_back_strap_l_02, swing1LimitDegrees=35, swing2LimitDegrees=35, twistLimitDegrees=15}
 {bone1Name=dyn_back_strap_l_04, bone2Name=dyn_back_strap_l_03, swing1LimitDegrees=35, swing2LimitDegrees=35, twistLimitDegrees=15}
 {
  bone1Name=dyn_back_strap_r_MSTR
  bone2Name=pelvis
  swing1LimitDegrees=45
  swing1Motion=Free
  swing2LimitDegrees=45
  swing2Motion=Free
  twistLimitDegrees=45
  twistMotion=Free
 }
 {bone1Name=dyn_back_strap_r_01, bone2Name=dyn_back_strap_r_MSTR, swing1LimitDegrees=35, swing2LimitDegrees=5, twistLimitDegrees=3}
 {bone1Name=dyn_back_strap_r_02, bone2Name=dyn_back_strap_r_01, swing1LimitDegrees=35, swing2LimitDegrees=20, twistLimitDegrees=15}
 {bone1Name=dyn_back_strap_r_03, bone2Name=dyn_back_strap_r_02, swing1LimitDegrees=35, swing2LimitDegrees=35, twistLimitDegrees=15}
 {bone1Name=dyn_back_strap_r_04, bone2Name=dyn_back_strap_r_03, swing1LimitDegrees=35, swing2LimitDegrees=35, twistLimitDegrees=15}
 {bone1Name=dyn_back_strap_r_05, bone2Name=dyn_back_strap_r_04, swing1LimitDegrees=35, swing2LimitDegrees=35, twistLimitDegrees=15}
 {bone1Name=dyn_calf_r_06, bone2Name=dyn_calf_r_MSTR_06, swing1LimitDegrees=35, swing2LimitDegrees=45, twistLimitDegrees=3}
 {
  bone1Name=dyn_calf_r_MSTR_01
  bone2Name=calf_twist_02_r
  swing1LimitDegrees=45
  swing1Motion=Free
  swing2LimitDegrees=45
  swing2Motion=Free
  twistLimitDegrees=45
  twistMotion=Free
 }
 {bone1Name=dyn_calf_r_01, bone2Name=dyn_calf_r_MSTR_01, swing1LimitDegrees=35, swing2LimitDegrees=45, twistLimitDegrees=3}
 {
  bone1Name=dyn_calf_r_MSTR_02
  bone2Name=calf_twist_02_r
  swing1LimitDegrees=45
  swing1Motion=Free
  swing2LimitDegrees=45
  swing2Motion=Free
  twistLimitDegrees=45
  twistMotion=Free
 }
 {bone1Name=dyn_calf_r_02, bone2Name=dyn_calf_r_MSTR_02, swing1LimitDegrees=35, swing2LimitDegrees=45, twistLimitDegrees=3}
 {
  bone1Name=dyn_calf_r_MSTR_03
  bone2Name=calf_twist_02_r
  swing1LimitDegrees=45
  swing1Motion=Free
  swing2LimitDegrees=45
  swing2Motion=Free
  twistLimitDegrees=45
  twistMotion=Free
 }
 {bone1Name=dyn_calf_r_03, bone2Name=dyn_calf_r_MSTR_03, swing1LimitDegrees=35, swing2LimitDegrees=45, twistLimitDegrees=3}
 {
  bone1Name=dyn_calf_r_MSTR_04
  bone2Name=calf_twist_02_r
  swing1LimitDegrees=45
  swing1Motion=Free
  swing2LimitDegrees=45
  swing2Motion=Free
  twistLimitDegrees=45
  twistMotion=Free
 }
 {bone1Name=dyn_calf_r_04, bone2Name=dyn_calf_r_MSTR_04, swing1LimitDegrees=35, swing2LimitDegrees=45, twistLimitDegrees=3}
 {
  bone1Name=dyn_calf_r_MSTR_05
  bone2Name=calf_twist_02_r
  swing1LimitDegrees=45
  swing1Motion=Free
  swing2LimitDegrees=45
  swing2Motion=Free
  twistLimitDegrees=45
  twistMotion=Free
 }
 {bone1Name=dyn_calf_r_05, bone2Name=dyn_calf_r_MSTR_05, swing1LimitDegrees=35, swing2LimitDegrees=45, twistLimitDegrees=3}
 {
  bone1Name=dyn_calf_r_MSTR_06
  bone2Name=calf_twist_02_r
  swing1LimitDegrees=45
  swing1Motion=Free
  swing2LimitDegrees=45
  swing2Motion=Free
  twistLimitDegrees=45
  twistMotion=Free
 }
 {bone1Name=dyn_calf_l_06, bone2Name=dyn_calf_l_MSTR_06, swing1LimitDegrees=35, swing2LimitDegrees=45, twistLimitDegrees=3}
 {
  bone1Name=dyn_calf_l_MSTR_01
  bone2Name=calf_twist_02_l
  swing1LimitDegrees=45
  swing1Motion=Free
  swing2LimitDegrees=45
  swing2Motion=Free
  twistLimitDegrees=45
  twistMotion=Free
 }
 {bone1Name=dyn_calf_l_01, bone2Name=dyn_calf_l_MSTR_01, swing1LimitDegrees=35, swing2LimitDegrees=45, twistLimitDegrees=3}
 {
  bone1Name=dyn_calf_l_MSTR_02
  bone2Name=calf_twist_02_l
  swing1LimitDegrees=45
  swing1Motion=Free
  swing2LimitDegrees=45
  swing2Motion=Free
  twistLimitDegrees=45
  twistMotion=Free
 }
 {bone1Name=dyn_calf_l_02, bone2Name=dyn_calf_l_MSTR_02, swing1LimitDegrees=35, swing2LimitDegrees=45, twistLimitDegrees=3}
 {
  bone1Name=dyn_calf_l_MSTR_03
  bone2Name=calf_twist_02_l
  swing1LimitDegrees=45
  swing1Motion=Free
  swing2LimitDegrees=45
  swing2Motion=Free
  twistLimitDegrees=45
  twistMotion=Free
 }
 {bone1Name=dyn_calf_l_03, bone2Name=dyn_calf_l_MSTR_03, swing1LimitDegrees=35, swing2LimitDegrees=45, twistLimitDegrees=3}
 {
  bone1Name=dyn_calf_l_MSTR_04
  bone2Name=calf_twist_02_l
  swing1LimitDegrees=45
  swing1Motion=Free
  swing2LimitDegrees=45
  swing2Motion=Free
  twistLimitDegrees=45
  twistMotion=Free
 }
 {bone1Name=dyn_calf_l_04, bone2Name=dyn_calf_l_MSTR_04, swing1LimitDegrees=35, swing2LimitDegrees=45, twistLimitDegrees=3}
 {
  bone1Name=dyn_calf_l_MSTR_05
  bone2Name=calf_twist_02_l
  swing1LimitDegrees=45
  swing1Motion=Free
  swing2LimitDegrees=45
  swing2Motion=Free
  twistLimitDegrees=45
  twistMotion=Free
 }
 {bone1Name=dyn_calf_l_05, bone2Name=dyn_calf_l_MSTR_05, swing1LimitDegrees=35, swing2LimitDegrees=45, twistLimitDegrees=3}
 {
  bone1Name=dyn_calf_l_MSTR_06
  bone2Name=calf_twist_02_l
  swing1LimitDegrees=45
  swing1Motion=Free
  swing2LimitDegrees=45
  swing2Motion=Free
  twistLimitDegrees=45
  twistMotion=Free
 }
 {
  bone1Name=thigh_twist_01_r
  bone2Name=thigh_r
  swing1LimitDegrees=45
  swing1Motion=Free
  swing2LimitDegrees=45
  swing2Motion=Free
  twistLimitDegrees=45
  twistMotion=Free
 }
 {
  bone1Name=thigh_twist_01_l
  bone2Name=thigh_l
  swing1LimitDegrees=45
  swing1Motion=Free
  swing2LimitDegrees=45
  swing2Motion=Free
  twistLimitDegrees=45
  twistMotion=Free
 }
 {
  bone1Name=calf_r
  bone2Name=thigh_r
  swing1LimitDegrees=45
  swing1Motion=Free
  swing2LimitDegrees=45
  swing2Motion=Free
  twistLimitDegrees=45
  twistMotion=Free
 }
 {
  bone1Name=calf_l
  bone2Name=thigh_l
  swing1LimitDegrees=45
  swing1Motion=Free
  swing2LimitDegrees=45
  swing2Motion=Free
  twistLimitDegrees=45
  twistMotion=Free
 }
]
