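# Physics asset for the skydive mesh of the medium female Veiled player body, covering dynamic loop bones that hang from both thighs.
# Capsules on the pelvis, thighs and calves form the body that the loops attach to.
# Each side has a two-segment main loop plus front and back loops, each anchored to the thigh by a tiny MSTR capsule.
# Main and back loops use narrower swing and tight twist limits, while front loops swing freely.
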PhysicsAsset /Game/Characters/Player/Female/Medium/Bodies/F_MED_Veiled/Meshes/F_MED_Veiled_skydive_Physics
bodies {
 calf_l [{shapeType=Capsule, length=31.75, radius=11.75, shapeName=calf_l_capsule, center=<-27 0 0>, rotation=<rot 90 80.7734 75.7735>}]
 calf_r [{shapeType=Capsule, length=31.75, radius=11.75, shapeName=calf_r_capsule, center=<27 0 0>, rotation=<rot 90 80.7735 75.7736>}]
 dyn_loop_l_01 [{shapeType=Capsule, length=7.2352, radius=2.7648, shapeName=dyn_loop_l_01_capsule, center=<5.2967 0 4>, rotation=<rot 0 0 90>}]
 dyn_loop_l_02 [
  {shapeType=Capsule, length=8.8242, radius=3.299, shapeName=dyn_loop_l_02_capsule, center=<7.7309 0 0>, rotation=<rot 0 0 90>}
  {
   shapeType=Capsule
   length=8.8388
   radius=1.1612
   shapeName=dyn_loop_l_02_capsule
   center=<1.074 8.1999 0.3555>
   rotation=<rot -87.8201 -119.6797 109.4485>
  }
 ]
 dyn_loop_l_MSTR [{shapeType=Capsule, length=0.505, radius=0.505, shapeName=dyn_loop_l_MSTR_capsule, rotation=<rot 0 -90 0>}]
 dyn_loop_l_bk [{shapeType=Capsule, length=5.9111, radius=2.5816, shapeName=dyn_loop_l_bk_capsule, center=<9.2031 0 0>, rotation=<rot 90 0 0>}]
 dyn_loop_l_bk_MSTR [{shapeType=Capsule, length=0.505, radius=0.505, shapeName=dyn_loop_l_bk_MSTR_capsule, rotation=<rot 0 -90 0>}]
 dyn_loop_l_fr [{
   shapeType=Capsule
   length=3.4551
   radius=1.5266
   shapeName=dyn_loop_l_fr_capsule
   center=<9.3035 0.382 0.5512>
   rotation=<rot 0 90.6965 -88.0762>
  }]
 dyn_loop_l_fr_MSTR [{shapeType=Capsule, length=0.505, radius=0.505, shapeName=dyn_loop_l_fr_MSTR_capsule, rotation=<rot 0 -90 0>}]
 dyn_loop_r_01 [{shapeType=Capsule, length=7.2352, radius=2.7648, shapeName=dyn_loop_r_01_capsule, center=<5.3 -4 0>}]
 dyn_loop_r_02 [
  {shapeType=Capsule, length=8.8242, radius=3.299, shapeName=dyn_loop_l_02_capsule, center=<7.7309 0 0>}
  {
   shapeType=Capsule
   length=8.8388
   radius=1.1612
   shapeName=dyn_loop_l_02_capsule
   center=<2.1529 -0.2805 -8.036>
   rotation=<rot 74.3167 -172.3225 -172.1679>
  }
 ]
 dyn_loop_r_MSTR [{shapeType=Capsule, length=0.505, radius=0.505, shapeName=dyn_loop_r_MSTR_capsule, rotation=<rot 0 -90 0>}]
 dyn_loop_r_bk [{shapeType=Capsule, length=5.5361, radius=2.9566, shapeName=dyn_loop_l_bk_capsule, center=<9.2031 0 0>, rotation=<rot 90 0 0>}]
 dyn_loop_r_bk_MSTR [{shapeType=Capsule, length=0.505, radius=0.505, shapeName=dyn_loop_r_bk_MSTR_capsule, rotation=<rot 0 -90 0>}]
 dyn_loop_r_fr [{
   shapeType=Capsule
   length=3.4552
   radius=1.5237
   shapeName=dyn_loop_r_fr_capsule
   center=<9.328 -0.0675 0.0765>
   rotation=<rot 0 -88.8477 -88.9012>
  }]
 dyn_loop_r_fr_MSTR [{shapeType=Capsule, length=0.505, radius=0.505, shapeName=dyn_loop_r_fr_MSTR_capsule, rotation=<rot 0 -90 0>}]
 pelvis [{shapeType=Capsule, length=10.5603, radius=10.1851, center=<-8.2175 -1.9871 0>}]
 thigh_l [
  {shapeType=Capsule, length=18.2231, radius=10, center=<-21.2717 -0.3685 1.6327>, rotation=<rot 90 89.4327 89.433>}
  {shapeType=Capsule, length=4.0449, radius=8.0801, shapeName=thigh_r_capsule, center=<-39.4665 0.7099 1.7776>, rotation=<rot 90 0 0>}
 ]
 thigh_r [
  {shapeType=Capsule, length=18.2231, radius=10, center=<21.2717 -0.3685 -1.6327>, rotation=<rot 90 89.4327 89.433>}
  {shapeType=Capsule, length=4.0449, radius=8.0801, shapeName=thigh_r_capsule, center=<39.4665 0.7099 -1.7776>, rotation=<rot 90 0 0>}
 ]
}
constraints [
 {
  bone1Name=thigh_r
  bone2Name=pelvis
  swing1LimitDegrees=45
  swing1Motion=Free
  swing2LimitDegrees=45
  swing2Motion=Free
  twistLimitDegrees=45
  twistMotion=Free
 }
 {bone1Name=thigh_l, bone2Name=pelvis, swing1LimitDegrees=45, swing2LimitDegrees=45, twistLimitDegrees=45}
 {
  bone1Name=calf_r
  bone2Name=thigh_r
  swing1LimitDegrees=45
  swing1Motion=Free
  swing2LimitDegrees=45
  swing2Motion=Free
  twistLimitDegrees=45
  twistMotion=Free
 }
 {
  bone1Name=calf_l
  bone2Name=thigh_l
  swing1LimitDegrees=45
  swing1Motion=Free
  swing2LimitDegrees=45
  swing2Motion=Free
  twistLimitDegrees=45
  twistMotion=Free
 }
 {
  bone1Name=dyn_loop_r_MSTR
  bone2Name=thigh_r
  swing1LimitDegrees=45
  swing1Motion=Free
  swing2LimitDegrees=45
  swing2Motion=Free
  twistLimitDegrees=45
  twistMotion=Free
 }
 {bone1Name=dyn_loop_r_01, bone2Name=dyn_loop_r_MSTR, swing1LimitDegrees=35, swing2LimitDegrees=35, twistLimitDegrees=3}
 {bone1Name=dyn_loop_r_02, bone2Name=dyn_loop_r_01, swing1LimitDegrees=35, swing2LimitDegrees=35, twistLimitDegrees=3}
 {
  bone1Name=dyn_loop_r_bk_MSTR
  bone2Name=thigh_r
  swing1LimitDegrees=45
  swing1Motion=Free
  swing2LimitDegrees=45
  swing2Motion=Free
  twistLimitDegrees=45
  twistMotion=Free
 }
 {bone1Name=dyn_loop_r_bk, bone2Name=dyn_loop_r_bk_MSTR, swing1LimitDegrees=35, swing2LimitDegrees=35, twistLimitDegrees=3}
 {
  bone1Name=dyn_loop_r_fr_MSTR
  bone2Name=thigh_r
  swing1LimitDegrees=45
  swing1Motion=Free
  swing2LimitDegrees=45
  swing2Motion=Free
  twistLimitDegrees=45
  twistMotion=Free
 }
 {
  bone1Name=dyn_loop_r_fr
  bone2Name=dyn_loop_r_fr_MSTR
  swing1LimitDegrees=45
  swing1Motion=Free
  swing2LimitDegrees=45
  swing2Motion=Free
  twistLimitDegrees=45
  twistMotion=Free
 }
 {
  bone1Name=dyn_loop_l_fr
  bone2Name=dyn_loop_l_fr_MSTR
  swing1LimitDegrees=45
  swing1Motion=Free
  swing2LimitDegrees=45
  swing2Motion=Free
  twistLimitDegrees=45
  twistMotion=Free
 }
 {
  bone1Name=dyn_loop_l_MSTR
  bone2Name=thigh_l
  swing1LimitDegrees=45
  swing1Motion=Free
  swing2LimitDegrees=45
  swing2Motion=Free
  twistLimitDegrees=45
  twistMotion=Free
 }
 {bone1Name=dyn_loop_l_01, bone2Name=dyn_loop_l_MSTR, swing1LimitDegrees=35, swing2LimitDegrees=35, twistLimitDegrees=3}
 {bone1Name=dyn_loop_l_02, bone2Name=dyn_loop_l_01, swing1LimitDegrees=35, swing2LimitDegrees=35, twistLimitDegrees=3}
 {
  bone1Name=dyn_loop_l_bk_MSTR
  bone2Name=thigh_l
  swing1LimitDegrees=45
  swing1Motion=Free
  swing2LimitDegrees=45
  swing2Motion=Free
  twistLimitDegrees=45
  twistMotion=Free
 }
 {bone1Name=dyn_loop_l_bk, bone2Name=dyn_loop_l_bk_MSTR, swing1LimitDegrees=35, swing2LimitDegrees=35, twistLimitDegrees=3}
 {
  bone1Name=dyn_loop_l_fr_MSTR
  bone2Name=thigh_l
  swing1LimitDegrees=45
  swing1Motion=Free
  swing2LimitDegrees=45
  swing2Motion=Free
  twistLimitDegrees=45
  twistMotion=Free
 }
]
