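# Animation Blueprint for the Veiled female player body that follows its parent mesh and adds secondary motion to clothing and accessories.
# Copied pose runs through deform and outfit Control Rigs plus weighted constraints on master bones.
# bRigidBodyEnabled selects RigidBody sims, varied by skydiving, cape and frontend state, or a fallback of bone offsets, Trail controllers and AnimDynamics.
# Shared AnimDynamics then drive sweater, calves, zipper, and hood unless wearing a backpack.
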AnimBlueprint /Game/Characters/Player/Female/Medium/Bodies/F_MED_Veiled/Meshes/F_MED_Veiled_AnimBP
target_skeleton=&/F_MED_Veiled_Skeleton
graphs [
 AnimGraph {nodes [
   Anim Dynamics - Bone: dyn_zipper!(ComponentPose, Alpha=0.25, GravityOverride:Vector="0,0,0", ExternalForce) -> (Pose) {
    angularDampingOverride=1
    angularSpringConstant=1500
    bAngularSpring=true
    bGravityOverrideInSimSpace=true
    bOverrideAngularDamping=true
    gravityScale=0.45
    simulationSpace=RootRelative
    alphaBoolBlend {blendOption=Linear}
    alphaScaleBias {scale=2}
    boundBone {boneName=dyn_zipper}
    componentAppliedLinearAccClamp=<2500 2500 2500>
    componentLinearAccScale=<1 1 1>
    componentLinearVelScale=<1 1 1>
    physicsBodyDefinitions [{
      sphereCollisionRadius=0
      boundBone {boneName=dyn_zipper}
      boxExtents=<10 5 5>
      constraintSetup {angularLimitsMax=<15 35 15>, angularLimitsMin=<-15 -35 -15>, angularTarget=<0 0 0>}
      localJointOffset=<2 0 0>
     }]
    relativeSpaceBone {boneName=root}
   }
   "Constraint - dyn_back_strap_r_05 (P)"(Alpha=1, ConstraintWeights_0=1) {
    alphaBoolBlend {blendOption=Linear}
    boneToModify {boneName=dyn_back_strap_r_05}
    constraintSetup [{targetBone {boneName=foot_r}}]
    constraintWeights [1]
   }
   n1 = Get bIsSkyDiving()
   n2 = Get GravityOverride()
   n3 = RigidBody!(
    ComponentPose=Local To Component(LocalPose=Use cached pose 'presim'())
    bAlphaBoolEnabled:bool=NOT Boolean(A=@n1)
    OverrideWorldGravity:Vector=@n2
    ExternalForce
   ) -> (Pose) {
    alphaInputType=Bool
    bClampLinearTranslationLimitToRefPose=true
    bOverrideWorldGravity=true
    overridePhysicsAsset=&/F_MED_Veiled_Physics
    simulationSpace=BaseBoneSpace
    alphaBoolBlend {blendOption=Linear}
    baseBoneRef {boneName=root}
    componentAppliedLinearAccClamp=<3500 3500 3500>
    componentLinearAccScale=<0.3 0.3 0.4>
    componentLinearVelScale=<0.3 0.3 0.4>
    overrideWorldGravity {z=-980}
   }
   n4 = RigidBody(ComponentPose=@n3, bAlphaBoolEnabled=@n1, OverrideWorldGravity=@n2) {
    alphaInputType=Bool
    bClampLinearTranslationLimitToRefPose=true
    bOverrideWorldGravity=true
    overridePhysicsAsset=&/F_MED_Veiled_skydive_Physics
    simulationSpace=BaseBoneSpace
    alphaBoolBlend {blendOption=Linear}
    baseBoneRef {boneName=pelvis}
    componentAppliedLinearAccClamp=<3500 3500 3500>
    componentLinearAccScale=<0.3 0.3 0.3>
    componentLinearVelScale=<0.3 0.3 0.3>
    overrideWorldGravity {z=-980}
   }
   n5 = Get Cape() -> (Cape:bool)
   n6 = RigidBody(ComponentPose=@n4, bAlphaBoolEnabled=NOT Boolean(A=@n5), OverrideWorldGravity=@n2) {
    alphaInputType=Bool
    bClampLinearTranslationLimitToRefPose=true
    bOverrideWorldGravity=true
    overridePhysicsAsset=&/F_MED_Veiled_Back_Physics
    simulationSpace=BaseBoneSpace
    alphaBoolBlend {blendInTime=0.1, blendOption=Linear, blendOutTime=0.1}
    baseBoneRef {boneName=root}
    componentAppliedLinearAccClamp=<3500 3500 3500>
    componentLinearAccScale=<0.3 0.3 0.4>
    componentLinearVelScale=<0.3 0.3 0.4>
    overrideWorldGravity {z=-980}
   }
   n7 = RigidBody(ComponentPose=@n6, bAlphaBoolEnabled=@n5, OverrideWorldGravity="0,0,-980") {
    alphaInputType=Bool
    bClampLinearTranslationLimitToRefPose=true
    bOverrideWorldGravity=true
    overridePhysicsAsset=&/F_MED_Veiled_Back_Cape_Physics
    simulationSpace=BaseBoneSpace
    alphaBoolBlend {blendInTime=0.1, blendOption=Linear, blendOutTime=0.1}
    baseBoneRef {boneName=root}
    componentAppliedLinearAccClamp=<3500 3500 3500>
    componentLinearAccScale=<0.3 0.3 0.4>
    componentLinearVelScale=<0.3 0.3 0.4>
    overrideWorldGravity {z=-980}
   }
   n8 = Component To Local(ComponentPose=@n7)
   n9 = Get GravityOverride()
   n10 = RigidBody!(ComponentPose=Local To Component(LocalPose=Use cached pose 'presim'()), Alpha:float, OverrideWorldGravity=@n9, ExternalForce) -> (Pose) {
    bClampLinearTranslationLimitToRefPose=true
    bOverrideWorldGravity=true
    overridePhysicsAsset=&F_MED_Veiled_Physics
    simulationSpace=BaseBoneSpace
    alphaBoolBlend {blendOption=Linear}
    baseBoneRef {boneName=root}
    componentAppliedLinearAccClamp=<3500 3500 3500>
    componentLinearAccScale=<0.3 0.3 0.4>
    componentLinearVelScale=<0.3 0.3 0.4>
    overrideWorldGravity {z=-980}
   }
   n11 = RigidBody!(ComponentPose=@n10, bAlphaBoolEnabled:bool, OverrideWorldGravity=@n9, ExternalForce) -> (Pose) {
    alphaInputType=Bool
    bClampLinearTranslationLimitToRefPose=true
    bOverrideWorldGravity=true
    overridePhysicsAsset=&/F_MED_Veiled_Back_Frontend_Physics
    simulationSpace=BaseBoneSpace
    alphaBoolBlend {blendInTime=0.1, blendOption=Linear, blendOutTime=0.1}
    baseBoneRef {boneName=root}
    componentAppliedLinearAccClamp=<3500 3500 3500>
    componentLinearAccScale=<0.3 0.3 0.4>
    componentLinearVelScale=<0.3 0.3 0.4>
    overrideWorldGravity {z=-980}
   }
   n12 = Component To Local(ComponentPose=@n11)
   n13 = Get bIsInFrontEnd()
   n14 = RigidBody!(
    ComponentPose=Local To Component(LocalPose=Blend Poses by bool(BlendPose_0=@n12, BlendPose_1=@n8, bActiveValue=@n13) {blendType=Linear})
    Alpha:float
    OverrideWorldGravity="0,0,-980"
    ExternalForce
   ) -> (Pose) {
    bClampLinearTranslationLimitToRefPose=true
    bOverrideWorldGravity=true
    overridePhysicsAsset=&/F_MED_Veiled_Arm_Physics
    simulationSpace=BaseBoneSpace
    alphaBoolBlend {blendOption=Linear}
    baseBoneRef {boneName=spine_02}
    componentLinearAccScale=<0.3 0.3 0.3>
    componentLinearVelScale=<0.3 0.3 0.3>
    overrideWorldGravity {z=-980}
   }
   n15 = Component To Local(ComponentPose=@n14)
   n16 = "Transform (Modify) Bone - Bone: dyn_arm_loop_02"(
    ComponentPose=Local To Component(LocalPose=Use cached pose 'presim'())
    Alpha=Get elbow R rot() -> (elbow R rot:double)
    Translation="0,0,0"
    Rotation="-55,0,0"
    Scale="1,1,1"
   ) {
    rotationMode=BMM_Additive
    rotationSpace=BCS_BoneSpace
    translationMode=BMM_Additive
    translationSpace=BCS_BoneSpace
    alphaBoolBlend {blendOption=Linear}
    boneToModify {boneName=dyn_arm_loop_02}
    rotation {pitch=-55}
   }
   n17 = Get still_crouch() -> (still_crouch:bool)
   n18 = "Transform (Modify) Bone - Bone: dyn_back_strap_l_02"!(ComponentPose=@n16, bAlphaBoolEnabled:bool=@n17, Translation="0,0,0", Rotation="0,-15,0", Scale="1,1,1") -> (Pose) {
    alphaInputType=Bool
    rotationMode=BMM_Additive
    rotationSpace=BCS_BoneSpace
    translationMode=BMM_Additive
    translationSpace=BCS_BoneSpace
    alphaBoolBlend {blendInTime=0.1, blendOption=Linear, blendOutTime=0.1}
    boneToModify {boneName=dyn_back_strap_l_02}
    rotation {yaw=-15}
   }
   n19 = "Transform (Modify) Bone - Bone: dyn_back_strap_l_03"!(ComponentPose=@n18, bAlphaBoolEnabled:bool=@n17, Translation="0,0,0", Rotation="0,-25,0", Scale="1,1,1") -> (Pose) {
    alphaInputType=Bool
    rotationMode=BMM_Additive
    rotationSpace=BCS_BoneSpace
    translationMode=BMM_Additive
    translationSpace=BCS_BoneSpace
    alphaBoolBlend {blendInTime=0.1, blendOption=Linear, blendOutTime=0.1}
    boneToModify {boneName=dyn_back_strap_l_03}
    rotation {yaw=-25}
   }
   n20 = "Transform (Modify) Bone - Bone: dyn_back_strap_l_04"!(ComponentPose=@n19, bAlphaBoolEnabled:bool=@n17, Translation="0,0,0", Rotation="0,-30,0", Scale="1,1,1") -> (Pose) {
    alphaInputType=Bool
    rotationMode=BMM_Additive
    rotationSpace=BCS_BoneSpace
    translationMode=BMM_Additive
    translationSpace=BCS_BoneSpace
    alphaBoolBlend {blendInTime=0.1, blendOption=Linear, blendOutTime=0.1}
    boneToModify {boneName=dyn_back_strap_l_04}
    rotation {yaw=-30}
   }
   n21 = "Transform (Modify) Bone - Bone: dyn_back_strap_r_02"!(ComponentPose=@n20, bAlphaBoolEnabled:bool=@n17, Translation="0,0,0", Rotation="0,-15,0", Scale="1,1,1") -> (Pose) {
    alphaInputType=Bool
    rotationMode=BMM_Additive
    rotationSpace=BCS_BoneSpace
    translationMode=BMM_Additive
    translationSpace=BCS_BoneSpace
    alphaBoolBlend {blendInTime=0.1, blendOption=Linear, blendOutTime=0.1}
    boneToModify {boneName=dyn_back_strap_r_02}
    rotation {yaw=-15}
   }
   n22 = "Transform (Modify) Bone - Bone: dyn_back_strap_r_03"!(ComponentPose=@n21, bAlphaBoolEnabled:bool=@n17, Translation="0,0,0", Rotation="0,-25,0", Scale="1,1,1") -> (Pose) {
    alphaInputType=Bool
    rotationMode=BMM_Additive
    rotationSpace=BCS_BoneSpace
    translationMode=BMM_Additive
    translationSpace=BCS_BoneSpace
    alphaBoolBlend {blendInTime=0.1, blendOption=Linear, blendOutTime=0.1}
    boneToModify {boneName=dyn_back_strap_r_03}
    rotation {yaw=-25}
   }
   n23 = "Transform (Modify) Bone - Bone: dyn_back_strap_r_04"!(ComponentPose=@n22, bAlphaBoolEnabled:bool=@n17, Translation="0,0,0", Rotation="0,0,0", Scale="1,1,1") -> (Pose) {
    alphaInputType=Bool
    rotationMode=BMM_Additive
    rotationSpace=BCS_BoneSpace
    translationMode=BMM_Additive
    translationSpace=BCS_BoneSpace
    alphaBoolBlend {blendInTime=0.1, blendOption=Linear, blendOutTime=0.1}
    boneToModify {boneName=dyn_back_strap_r_04}
   }
   n24 = "Transform (Modify) Bone - Bone: dyn_back_strap_l_01"(ComponentPose=@n23, Alpha=Get ThighRotL() -> (ThighRotL:double), Translation="0,0,0", Rotation="15,-15,0", Scale="1,1,1") {
    rotationMode=BMM_Additive
    rotationSpace=BCS_BoneSpace
    translationMode=BMM_Additive
    translationSpace=BCS_BoneSpace
    alphaBoolBlend {blendInTime=0.1, blendOption=Linear, blendOutTime=0.1}
    alphaScaleBiasClamp {bInterpResult=true}
    boneToModify {boneName=dyn_back_strap_l_01}
    rotation {pitch=15, yaw=-15}
   }
   n25 = "Transform (Modify) Bone - Bone: dyn_back_strap_r_01"(ComponentPose=@n24, Alpha=Get ThighRotR() -> (ThighRotR:double), Translation="0,0,0", Rotation="-15,-15,0", Scale="1,1,1") {
    rotationMode=BMM_Additive
    rotationSpace=BCS_BoneSpace
    translationMode=BMM_Additive
    translationSpace=BCS_BoneSpace
    alphaBoolBlend {blendInTime=0.1, blendOption=Linear, blendOutTime=0.1}
    alphaScaleBiasClamp {bInterpResult=true}
    boneToModify {boneName=dyn_back_strap_r_01}
    rotation {pitch=-15, yaw=-15}
   }
   n26 = Get BwdMovement() -> (BwdMovement:float)
   n27 = Trail controller - Bone: dyn_back_strap_r_05!(ComponentPose=@n25, Alpha=@n26, FakeVelocity:Vector="0,0,0") -> (Pose) {
    bActorSpaceFakeVel=true
    bShowBaseMotion=false
    bShowLimit=false
    chainLength=4
    debugLifeTime=0.1
    maxDeltaTime=0.0333
    relaxationSpeedScale=3
    alphaScaleBiasClamp {bInterpResult=true}
    baseJoint {boneName=root}
    planarLimits [
     {drivingBone {boneName=root}, planeTransform {location=<0 0 10>, rotation=<rot 0 0 0>, scale=<1 1 1>}}
     {drivingBone {boneName=spine_03}, planeTransform {location=<0 -20 0>, rotation=<rot 0 15 -89.9999>, scale=<1 1 1>}}
    ]
    rotationLimits [{limitMax=<180 180 180>}, {limitMax=<180 180 180>}, {limitMax=<180 180 180>}, {limitMax=<180 180 180>}]
    rotationOffsets [<0 0 0>, <0 0 0>, <0 0 0>, <0 0 0>]
    trailBone {boneName=dyn_back_strap_r_05}
    trailRelaxationSpeed {editorCurveData {keys [{arriveTangent=-72.0739, interpMode=RCIM_Cubic, leaveTangent=-72.0739, tangentMode=RCTM_User, time=0, value=35}, <at 1 5 RCIM_Cubic>]}}
   }
   n28 = Trail controller - Bone: dyn_back_strap_l_04!(ComponentPose=@n27, Alpha=@n26, FakeVelocity:Vector="0,0,0") -> (Pose) {
    bActorSpaceFakeVel=true
    bShowBaseMotion=false
    bShowLimit=false
    chainLength=3
    debugLifeTime=0.1
    maxDeltaTime=0.0333
    relaxationSpeedScale=3
    alphaScaleBiasClamp {bInterpResult=true}
    baseJoint {boneName=root}
    planarLimits [
     {drivingBone {boneName=root}, planeTransform {location=<0 0 10>, rotation=<rot 0 0 0>, scale=<1 1 1>}}
     {drivingBone {boneName=spine_03}, planeTransform {location=<0 -20 0>, rotation=<rot 0 15 -89.9999>, scale=<1 1 1>}}
    ]
    rotationLimits [{limitMax=<180 180 180>}, {limitMax=<180 180 180>}, {limitMax=<180 180 180>}]
    rotationOffsets [<0 0 0>, <0 0 0>, <0 0 0>]
    trailBone {boneName=dyn_back_strap_l_04}
    trailRelaxationSpeed {editorCurveData {keys [{arriveTangent=-72.0739, interpMode=RCIM_Cubic, leaveTangent=-72.0739, tangentMode=RCTM_User, time=0, value=35}, <at 1 5 RCIM_Cubic>]}}
   }
   n29 = Anim Dynamics - Bone: dyn_loop_l_01(ComponentPose=@n28, Alpha=1, ExternalForce="0,0,0") {
    angularDampingOverride=0.9999
    angularSpringConstant=150
    bAngularSpring=true
    bGravityOverrideInSimSpace=true
    bOverrideAngularDamping=true
    simulationSpace=RootRelative
    alphaBoolBlend {blendOption=Linear}
    boundBone {boneName=dyn_loop_l_01}
    componentAppliedLinearAccClamp=<3500 3500 3500>
    componentLinearAccScale=<3 3 3>
    componentLinearVelScale=<3 3 3>
    physicsBodyDefinitions [{
      sphereCollisionRadius=0
      boundBone {boneName=dyn_loop_l_01}
      boxExtents=<10 20 20>
      constraintSetup {angularLimitsMax=<5 5 15>, angularLimitsMin=<-5 -35 -15>, angularTarget=<0 0 0>}
      localJointOffset=<2 0 -0.25>
     }]
   }
   n30 = Anim Dynamics - Bone: dyn_loop_l_02(ComponentPose=@n29, Alpha=1, ExternalForce="0,0,0") {
    angularDampingOverride=0.9999
    angularSpringConstant=150
    bAngularSpring=true
    bGravityOverrideInSimSpace=true
    bOverrideAngularDamping=true
    simulationSpace=RootRelative
    alphaBoolBlend {blendOption=Linear}
    boundBone {boneName=dyn_loop_l_02}
    componentAppliedLinearAccClamp=<3500 3500 3500>
    componentLinearAccScale=<3 3 3>
    componentLinearVelScale=<3 3 3>
    physicsBodyDefinitions [{
      sphereCollisionRadius=0
      boundBone {boneName=dyn_loop_l_02}
      boxExtents=<10 20 20>
      constraintSetup {angularLimitsMax=<5 5 15>, angularLimitsMin=<-5 -35 -15>, angularTarget=<0 0 0>}
      localJointOffset=<2 0 -0.25>
     }]
   }
   n31 = Anim Dynamics - Bone: dyn_loop_r_01(ComponentPose=@n30, Alpha=1, ExternalForce="0,0,0") {
    angularDampingOverride=0.9999
    angularSpringConstant=150
    bAngularSpring=true
    bGravityOverrideInSimSpace=true
    bOverrideAngularDamping=true
    simulationSpace=RootRelative
    alphaBoolBlend {blendOption=Linear}
    boundBone {boneName=dyn_loop_r_01}
    componentAppliedLinearAccClamp=<3500 3500 3500>
    componentLinearAccScale=<3 3 3>
    componentLinearVelScale=<3 3 3>
    physicsBodyDefinitions [{
      sphereCollisionRadius=0
      boundBone {boneName=dyn_loop_r_01}
      boxExtents=<10 20 20>
      constraintSetup {angularLimitsMax=<5 15 5>, angularLimitsMin=<-5 -15 -35>, angularTarget=<0 0 0>}
      localJointOffset=<2 0 -0.25>
     }]
   }
   n32 = Anim Dynamics - Bone: dyn_loop_r_02(ComponentPose=@n31, Alpha=1, ExternalForce="0,0,0") {
    angularDampingOverride=0.9999
    angularSpringConstant=150
    bAngularSpring=true
    bGravityOverrideInSimSpace=true
    bOverrideAngularDamping=true
    simulationSpace=RootRelative
    alphaBoolBlend {blendOption=Linear}
    boundBone {boneName=dyn_loop_r_02}
    componentAppliedLinearAccClamp=<3500 3500 3500>
    componentLinearAccScale=<3 3 3>
    componentLinearVelScale=<3 3 3>
    physicsBodyDefinitions [{
      sphereCollisionRadius=0
      boundBone {boneName=dyn_loop_r_02}
      boxExtents=<10 20 20>
      constraintSetup {angularLimitsMax=<5 15 5>, angularLimitsMin=<-5 -15 -35>, angularTarget=<0 0 0>}
      localJointOffset=<2 0 -0.25>
     }]
   }
   n33 = Anim Dynamics - Bone: dyn_arm_loop_01(ComponentPose=@n32, Alpha=1, ExternalForce="0,0,0") {
    angularDampingOverride=0.9999
    angularSpringConstant=150
    bAngularSpring=true
    bGravityOverrideInSimSpace=true
    bOverrideAngularDamping=true
    simulationSpace=RootRelative
    alphaBoolBlend {blendOption=Linear}
    boundBone {boneName=dyn_arm_loop_01}
    componentAppliedLinearAccClamp=<3500 3500 3500>
    componentLinearAccScale=<3 3 3>
    componentLinearVelScale=<3 3 3>
    physicsBodyDefinitions [{
      sphereCollisionRadius=0
      boundBone {boneName=dyn_arm_loop_01}
      boxExtents=<10 20 20>
      constraintSetup {angularLimitsMax=<5 0 10>, angularLimitsMin=<-5 -25 -10>, angularTarget=<0 0 0>}
      localJointOffset=<2 0 -0.25>
     }]
   }
   n34 = Anim Dynamics - Bone: dyn_arm_loop_02(ComponentPose=@n33, Alpha=1, ExternalForce="0,0,0") {
    angularDampingOverride=0.9999
    angularSpringConstant=150
    bAngularSpring=true
    bGravityOverrideInSimSpace=true
    bOverrideAngularDamping=true
    simulationSpace=RootRelative
    alphaBoolBlend {blendOption=Linear}
    boundBone {boneName=dyn_arm_loop_02}
    componentAppliedLinearAccClamp=<3500 3500 3500>
    componentLinearAccScale=<3 3 3>
    componentLinearVelScale=<3 3 3>
    physicsBodyDefinitions [{
      sphereCollisionRadius=0
      boundBone {boneName=dyn_arm_loop_02}
      boxExtents=<10 20 20>
      constraintSetup {angularLimitsMax=<5 0 10>, angularLimitsMin=<-5 -25 -10>, angularTarget=<0 0 0>}
      localJointOffset=<2 0 -0.25>
     }]
   }
   n35 = Component To Local(ComponentPose=@n34)
   n36 = Get bRigidBodyEnabled()
   n37 = Anim Dynamics - Bone: dyn_sweater_bck_l(
    ComponentPose=Local To Component(LocalPose=Blend Poses by bool(BlendPose_0=@n15, BlendPose_1=@n35, bActiveValue=@n36) {blendType=Linear})
    Alpha=1
    ExternalForce="0,0,0"
   ) {
    angularDampingOverride=0.97
    angularSpringConstant=50
    bAngularSpring=true
    bGravityOverrideInSimSpace=true
    bOverrideAngularDamping=true
    simulationSpace=RootRelative
    alphaBoolBlend {blendOption=Linear}
    boundBone {boneName=dyn_sweater_bck_l}
    componentAppliedLinearAccClamp=<3500 3500 3500>
    componentLinearAccScale=<1 1 1>
    componentLinearVelScale=<1 1 1>
    gravityOverride {z=-810}
    physicsBodyDefinitions [{
      sphereCollisionRadius=0
      boundBone {boneName=dyn_sweater_bck_l}
      boxExtents=<20 20 10>
      constraintSetup {angularLimitsMax=<5 5 15>, angularLimitsMin=<-5 -35 0>, angularTarget=<0 0 0>}
      localJointOffset=<3 0 0>
     }]
    relativeSpaceBone {boneName=root}
   }
   n38 = Anim Dynamics - Bone: dyn_sweater_bck_r(ComponentPose=@n37, Alpha=1, ExternalForce="0,0,0") {
    angularDampingOverride=0.97
    angularSpringConstant=50
    bAngularSpring=true
    bGravityOverrideInSimSpace=true
    bOverrideAngularDamping=true
    simulationSpace=RootRelative
    alphaBoolBlend {blendOption=Linear}
    boundBone {boneName=dyn_sweater_bck_r}
    componentAppliedLinearAccClamp=<3500 3500 3500>
    componentLinearAccScale=<1 1 1>
    componentLinearVelScale=<1 1 1>
    gravityOverride {z=-810}
    physicsBodyDefinitions [{
      sphereCollisionRadius=0
      boundBone {boneName=dyn_sweater_bck_r}
      boxExtents=<20 20 10>
      constraintSetup {angularLimitsMax=<5 5 0>, angularLimitsMin=<-5 -35 15>, angularTarget=<0 0 0>}
      localJointOffset=<3 0 0>
     }]
    relativeSpaceBone {boneName=root}
   }
   n39 = Anim Dynamics - Bone: dyn_sweater_bck(ComponentPose=@n38, Alpha=1, ExternalForce="0,0,0") {
    angularDampingOverride=0.97
    angularSpringConstant=50
    bAngularSpring=true
    bGravityOverrideInSimSpace=true
    bOverrideAngularDamping=true
    simulationSpace=RootRelative
    alphaBoolBlend {blendOption=Linear}
    boundBone {boneName=dyn_sweater_bck}
    componentAppliedLinearAccClamp=<3500 3500 3500>
    componentLinearAccScale=<1 1 1>
    componentLinearVelScale=<1 1 1>
    gravityOverride {z=-810}
    physicsBodyDefinitions [{
      sphereCollisionRadius=0
      boundBone {boneName=dyn_sweater_bck}
      boxExtents=<20 20 10>
      constraintSetup {angularLimitsMax=<5 10 10>, angularLimitsMin=<-5 -45 -10>, angularTarget=<0 0 0>}
      localJointOffset=<3 0 0>
     }]
    relativeSpaceBone {boneName=root}
   }
   n40 = Anim Dynamics - Bone: dyn_sweater_side_l(ComponentPose=@n39, Alpha=1, ExternalForce="0,0,0") {
    angularDampingOverride=0.98
    angularSpringConstant=50
    bAngularSpring=true
    bGravityOverrideInSimSpace=true
    bOverrideAngularDamping=true
    simulationSpace=RootRelative
    alphaBoolBlend {blendOption=Linear}
    boundBone {boneName=dyn_sweater_side_l}
    componentAppliedLinearAccClamp=<3500 3500 3500>
    componentLinearAccScale=<0.8 0.8 0.8>
    componentLinearVelScale=<0.8 0.8 0.8>
    gravityOverride {z=-810}
    physicsBodyDefinitions [{
      sphereCollisionRadius=0
      boundBone {boneName=dyn_sweater_side_l}
      boxExtents=<20 20 10>
      constraintSetup {angularLimitsMax=<5 0 5>, angularLimitsMin=<-5 -35 -5>, angularTarget=<0 0 0>}
      localJointOffset=<3 0 0>
     }]
    relativeSpaceBone {boneName=root}
   }
   n41 = Anim Dynamics - Bone: dyn_sweater_side_r(ComponentPose=@n40, Alpha=1, ExternalForce="0,0,0") {
    angularDampingOverride=0.98
    angularSpringConstant=50
    bAngularSpring=true
    bGravityOverrideInSimSpace=true
    bOverrideAngularDamping=true
    simulationSpace=RootRelative
    alphaBoolBlend {blendOption=Linear}
    boundBone {boneName=dyn_sweater_side_r}
    componentAppliedLinearAccClamp=<3500 3500 3500>
    componentLinearAccScale=<0.8 0.8 0.8>
    componentLinearVelScale=<0.8 0.8 0.8>
    gravityOverride {z=-810}
    physicsBodyDefinitions [{
      sphereCollisionRadius=0
      boundBone {boneName=dyn_sweater_side_r}
      boxExtents=<20 20 10>
      constraintSetup {angularLimitsMax=<5 0 5>, angularLimitsMin=<-5 -35 -5>, angularTarget=<0 0 0>}
      localJointOffset=<3 0 0>
     }]
    relativeSpaceBone {boneName=root}
   }
   n42 = Anim Dynamics - Bone: dyn_calf_l_01(ComponentPose=@n41, Alpha=0.5, ExternalForce="0,0,0") {
    angularDampingOverride=0.9999
    angularSpringConstant=1250
    bAngularSpring=true
    bGravityOverrideInSimSpace=true
    bOverrideAngularDamping=true
    gravityScale=0.8
    simulationSpace=BoneRelative
    alphaBoolBlend {blendOption=Linear}
    boundBone {boneName=dyn_calf_l_01}
    componentAppliedLinearAccClamp=<3500 3500 3500>
    componentLinearAccScale=<1.5 1.5 1.25>
    componentLinearVelScale=<1.5 1.5 1.25>
    gravityOverride {z=-980}
    physicsBodyDefinitions [{
      sphereCollisionRadius=0
      boundBone {boneName=dyn_calf_l_01}
      boxExtents=<17 17 10>
      constraintSetup {angularLimitsMax=<5 0 15>, angularLimitsMin=<-5 -75 -15>, angularTarget=<1 0 0>}
      localJointOffset=<3 0 0>
     }]
    relativeSpaceBone {boneName=root}
   }
   n43 = Anim Dynamics - Bone: dyn_calf_l_02(ComponentPose=@n42, Alpha=0.5, ExternalForce="0,0,0") {
    angularDampingOverride=0.9999
    angularSpringConstant=1500
    bAngularSpring=true
    bGravityOverrideInSimSpace=true
    bOverrideAngularDamping=true
    gravityScale=0.8
    simulationSpace=BoneRelative
    alphaBoolBlend {blendOption=Linear}
    boundBone {boneName=dyn_calf_l_02}
    componentAppliedLinearAccClamp=<3500 3500 3500>
    componentLinearAccScale=<1.5 1.5 1.25>
    componentLinearVelScale=<1.5 1.5 1.25>
    gravityOverride {z=-980}
    physicsBodyDefinitions [{
      sphereCollisionRadius=0
      boundBone {boneName=dyn_calf_l_02}
      boxExtents=<17 17 10>
      constraintSetup {angularLimitsMax=<5 0 15>, angularLimitsMin=<-5 -75 -15>, angularTarget=<1 0 0>}
      localJointOffset=<4 0.5 0>
     }]
    relativeSpaceBone {boneName=root}
   }
   n44 = Anim Dynamics - Bone: dyn_calf_l_03(ComponentPose=@n43, Alpha=0.5, ExternalForce="0,0,0") {
    angularDampingOverride=0.9999
    angularSpringConstant=1250
    bAngularSpring=true
    bGravityOverrideInSimSpace=true
    bOverrideAngularDamping=true
    gravityScale=0.8
    simulationSpace=BoneRelative
    alphaBoolBlend {blendOption=Linear}
    boundBone {boneName=dyn_calf_l_03}
    componentAppliedLinearAccClamp=<3500 3500 3500>
    componentLinearAccScale=<1.5 1.5 1.25>
    componentLinearVelScale=<1.5 1.5 1.25>
    gravityOverride {z=-980}
    physicsBodyDefinitions [{
      sphereCollisionRadius=0
      boundBone {boneName=dyn_calf_l_03}
      boxExtents=<17 17 10>
      constraintSetup {angularLimitsMax=<5 0 15>, angularLimitsMin=<-5 -75 -15>, angularTarget=<1 0 0>}
      localJointOffset=<3 1 0>
     }]
    relativeSpaceBone {boneName=root}
   }
   n45 = Anim Dynamics - Bone: dyn_calf_l_04(ComponentPose=@n44, Alpha=0.5, ExternalForce="0,0,0") {
    angularDampingOverride=0.9999
    angularSpringConstant=1500
    bAngularSpring=true
    bGravityOverrideInSimSpace=true
    bOverrideAngularDamping=true
    gravityScale=0.8
    simulationSpace=BoneRelative
    alphaBoolBlend {blendOption=Linear}
    boundBone {boneName=dyn_calf_l_04}
    componentAppliedLinearAccClamp=<3500 3500 3500>
    componentLinearAccScale=<1.5 1.5 1.25>
    componentLinearVelScale=<1.5 1.5 1.25>
    gravityOverride {z=-980}
    physicsBodyDefinitions [{
      sphereCollisionRadius=0
      boundBone {boneName=dyn_calf_l_04}
      boxExtents=<17 17 10>
      constraintSetup {angularLimitsMax=<5 75 15>, angularLimitsMin=<-5 0 -15>, angularTarget=<1 0 0>}
      localJointOffset=<4 0 0>
     }]
    relativeSpaceBone {boneName=root}
   }
   n46 = Anim Dynamics - Bone: dyn_calf_l_05(ComponentPose=@n45, Alpha=0.5, ExternalForce="0,0,0") {
    angularDampingOverride=0.9999
    angularSpringConstant=1250
    bAngularSpring=true
    bGravityOverrideInSimSpace=true
    bOverrideAngularDamping=true
    gravityScale=0.8
    simulationSpace=BoneRelative
    alphaBoolBlend {blendOption=Linear}
    boundBone {boneName=dyn_calf_l_05}
    componentAppliedLinearAccClamp=<3500 3500 3500>
    componentLinearAccScale=<1.5 1.5 1.25>
    componentLinearVelScale=<1.5 1.5 1.25>
    gravityOverride {z=-980}
    physicsBodyDefinitions [{
      sphereCollisionRadius=0
      boundBone {boneName=dyn_calf_l_05}
      boxExtents=<17 17 10>
      constraintSetup {angularLimitsMax=<5 0 15>, angularLimitsMin=<-5 -75 -15>, angularTarget=<1 0 0>}
      localJointOffset=<3 0 0>
     }]
    relativeSpaceBone {boneName=root}
   }
   n47 = Anim Dynamics - Bone: dyn_calf_l_06(ComponentPose=@n46, Alpha=0.5, ExternalForce="0,0,0") {
    angularDampingOverride=0.9999
    angularSpringConstant=1500
    bAngularSpring=true
    bGravityOverrideInSimSpace=true
    bOverrideAngularDamping=true
    gravityScale=0.8
    simulationSpace=BoneRelative
    alphaBoolBlend {blendOption=Linear}
    boundBone {boneName=dyn_calf_l_06}
    componentAppliedLinearAccClamp=<3500 3500 3500>
    componentLinearAccScale=<1.5 1.5 1.25>
    componentLinearVelScale=<1.5 1.5 1.25>
    gravityOverride {z=-980}
    physicsBodyDefinitions [{
      sphereCollisionRadius=0
      boundBone {boneName=dyn_calf_l_06}
      boxExtents=<17 17 10>
      constraintSetup {angularLimitsMax=<5 0 15>, angularLimitsMin=<-5 -75 -15>, angularTarget=<1 0 0>}
      localJointOffset=<4 0 0>
     }]
    relativeSpaceBone {boneName=root}
   }
   n48 = Anim Dynamics - Bone: dyn_calf_r_01(ComponentPose=@n47, Alpha=0.5, ExternalForce="0,0,0") {
    angularDampingOverride=0.9999
    angularSpringConstant=1250
    bAngularSpring=true
    bGravityOverrideInSimSpace=true
    bOverrideAngularDamping=true
    gravityScale=0.8
    simulationSpace=BoneRelative
    alphaBoolBlend {blendOption=Linear}
    boundBone {boneName=dyn_calf_r_01}
    componentAppliedLinearAccClamp=<3500 3500 3500>
    componentLinearAccScale=<1.5 1.5 1.25>
    componentLinearVelScale=<1.5 1.5 1.25>
    gravityOverride {z=-980}
    physicsBodyDefinitions [{
      sphereCollisionRadius=0
      boundBone {boneName=dyn_calf_r_01}
      boxExtents=<17 17 10>
      constraintSetup {angularLimitsMax=<5 0 15>, angularLimitsMin=<-5 -75 -15>, angularTarget=<1 0 0>}
      localJointOffset=<3 -0.5 0>
     }]
    relativeSpaceBone {boneName=root}
   }
   n49 = Anim Dynamics - Bone: dyn_calf_r_02(ComponentPose=@n48, Alpha=0.5, ExternalForce="0,0,0") {
    angularDampingOverride=0.9999
    angularSpringConstant=1500
    bAngularSpring=true
    bGravityOverrideInSimSpace=true
    bOverrideAngularDamping=true
    gravityScale=0.8
    simulationSpace=BoneRelative
    alphaBoolBlend {blendOption=Linear}
    boundBone {boneName=dyn_calf_r_02}
    componentAppliedLinearAccClamp=<3500 3500 3500>
    componentLinearAccScale=<1.5 1.5 1.25>
    componentLinearVelScale=<1.5 1.5 1.25>
    gravityOverride {z=-980}
    physicsBodyDefinitions [{
      sphereCollisionRadius=0
      boundBone {boneName=dyn_calf_r_02}
      boxExtents=<17 17 10>
      constraintSetup {angularLimitsMax=<5 0 15>, angularLimitsMin=<-5 -75 -15>, angularTarget=<1 0 0>}
      localJointOffset=<4 0.5 0>
     }]
    relativeSpaceBone {boneName=root}
   }
   n50 = Anim Dynamics - Bone: dyn_calf_r_03(ComponentPose=@n49, Alpha=0.5, ExternalForce="0,0,0") {
    angularDampingOverride=0.9999
    angularSpringConstant=1250
    bAngularSpring=true
    bGravityOverrideInSimSpace=true
    bOverrideAngularDamping=true
    gravityScale=0.8
    simulationSpace=BoneRelative
    alphaBoolBlend {blendOption=Linear}
    boundBone {boneName=dyn_calf_r_03}
    componentAppliedLinearAccClamp=<3500 3500 3500>
    componentLinearAccScale=<1.5 1.5 1.25>
    componentLinearVelScale=<1.5 1.5 1.25>
    gravityOverride {z=-980}
    physicsBodyDefinitions [{
      sphereCollisionRadius=0
      boundBone {boneName=dyn_calf_r_03}
      boxExtents=<17 17 10>
      constraintSetup {angularLimitsMax=<5 0 15>, angularLimitsMin=<-5 -75 -15>, angularTarget=<1 0 0>}
      localJointOffset=<3 1 0>
     }]
    relativeSpaceBone {boneName=root}
   }
   n51 = Anim Dynamics - Bone: dyn_calf_r_04(ComponentPose=@n50, Alpha=0.5, ExternalForce="0,0,0") {
    angularDampingOverride=0.9999
    angularSpringConstant=1500
    bAngularSpring=true
    bGravityOverrideInSimSpace=true
    bOverrideAngularDamping=true
    gravityScale=0.8
    simulationSpace=BoneRelative
    alphaBoolBlend {blendOption=Linear}
    boundBone {boneName=dyn_calf_r_04}
    componentAppliedLinearAccClamp=<3500 3500 3500>
    componentLinearAccScale=<1.5 1.5 1.25>
    componentLinearVelScale=<1.5 1.5 1.25>
    gravityOverride {z=-980}
    physicsBodyDefinitions [{
      sphereCollisionRadius=0
      boundBone {boneName=dyn_calf_r_04}
      boxExtents=<17 17 10>
      constraintSetup {angularLimitsMax=<5 0 15>, angularLimitsMin=<-5 -75 -15>, angularTarget=<1 0 0>}
      localJointOffset=<4 0 0>
     }]
    relativeSpaceBone {boneName=root}
   }
   n52 = Anim Dynamics - Bone: dyn_calf_r_05(ComponentPose=@n51, Alpha=0.5, ExternalForce="0,0,0") {
    angularDampingOverride=0.9999
    angularSpringConstant=1250
    bAngularSpring=true
    bGravityOverrideInSimSpace=true
    bOverrideAngularDamping=true
    gravityScale=0.8
    simulationSpace=BoneRelative
    alphaBoolBlend {blendOption=Linear}
    boundBone {boneName=dyn_calf_r_05}
    componentAppliedLinearAccClamp=<3500 3500 3500>
    componentLinearAccScale=<1.5 1.5 1.25>
    componentLinearVelScale=<1.5 1.5 1.25>
    gravityOverride {z=-980}
    physicsBodyDefinitions [{
      sphereCollisionRadius=0
      boundBone {boneName=dyn_calf_r_05}
      boxExtents=<17 17 10>
      constraintSetup {angularLimitsMax=<5 0 15>, angularLimitsMin=<-5 -75 -15>, angularTarget=<1 0 0>}
      localJointOffset=<3 0 0>
     }]
    relativeSpaceBone {boneName=root}
   }
   n53 = Anim Dynamics - Bone: dyn_calf_r_06(ComponentPose=@n52, Alpha=0.5, ExternalForce="0,0,0") {
    angularDampingOverride=0.9999
    angularSpringConstant=1500
    bAngularSpring=true
    bGravityOverrideInSimSpace=true
    bOverrideAngularDamping=true
    gravityScale=0.8
    simulationSpace=BoneRelative
    alphaBoolBlend {blendOption=Linear}
    boundBone {boneName=dyn_calf_r_06}
    componentAppliedLinearAccClamp=<3500 3500 3500>
    componentLinearAccScale=<1.5 1.5 1.25>
    componentLinearVelScale=<1.5 1.5 1.25>
    gravityOverride {z=-980}
    physicsBodyDefinitions [{
      sphereCollisionRadius=0
      boundBone {boneName=dyn_calf_r_06}
      boxExtents=<17 17 10>
      constraintSetup {angularLimitsMax=<5 0 15>, angularLimitsMin=<-5 -75 -15>, angularTarget=<1 0 0>}
      localJointOffset=<4 0 0>
     }]
    relativeSpaceBone {boneName=root}
   }
   n54 = NOT Boolean(A=Get backpack() -> (backpack:bool))
   n55 = Anim Dynamics - Bone: dyn_hood_01!(ComponentPose=@n53, bAlphaBoolEnabled:bool=@n54, ExternalForce="0,0,0") -> (Pose) {
    alphaInputType=Bool
    angularDampingOverride=0.9999
    angularSpringConstant=50
    bAngularSpring=true
    bGravityOverrideInSimSpace=true
    bOverrideAngularDamping=true
    simulationSpace=RootRelative
    alphaBoolBlend {blendInTime=0.1, blendOption=Linear, blendOutTime=0.1}
    boundBone {boneName=dyn_hood_01}
    componentAppliedLinearAccClamp=<3500 3500 3500>
    componentLinearAccScale=<0.8 0.8 0.8>
    componentLinearVelScale=<0.8 0.8 0.8>
    gravityOverride {z=-810}
    physicsBodyDefinitions [{
      sphereCollisionRadius=0
      boundBone {boneName=dyn_hood_01}
      boxExtents=<20 20 10>
      constraintSetup {angularLimitsMax=<5 10 5>, angularLimitsMin=<-5 -30 -5>, angularTarget=<1 0 0>}
      localJointOffset=<3 0 0>
     }]
    relativeSpaceBone {boneName=root}
   }
   n56 = Anim Dynamics - Bone: dyn_hood_02!(ComponentPose=@n55, bAlphaBoolEnabled:bool=@n54, ExternalForce="0,0,0") -> (Pose) {
    alphaInputType=Bool
    angularDampingOverride=0.9999
    angularSpringConstant=50
    bAngularSpring=true
    bGravityOverrideInSimSpace=true
    bOverrideAngularDamping=true
    simulationSpace=RootRelative
    alphaBoolBlend {blendInTime=0.1, blendOption=Linear, blendOutTime=0.1}
    boundBone {boneName=dyn_hood_02}
    componentAppliedLinearAccClamp=<3500 3500 3500>
    componentLinearAccScale=<0.8 0.8 0.8>
    componentLinearVelScale=<0.8 0.8 0.8>
    gravityOverride {z=-810}
    physicsBodyDefinitions [{
      sphereCollisionRadius=0
      boundBone {boneName=dyn_hood_02}
      boxExtents=<20 20 10>
      constraintSetup {angularLimitsMax=<5 10 5>, angularLimitsMin=<-5 -30 -5>, angularTarget=<1 0 0>}
      localJointOffset=<3 0 0>
     }]
    relativeSpaceBone {boneName=root}
   }
   n57 = Anim Dynamics - Bone: dyn_zipper!(ComponentPose=@n56, Alpha=1, ExternalForce) -> (Pose) {
    angularDampingOverride=0.98
    angularSpringConstant=300
    bAngularSpring=true
    bGravityOverrideInSimSpace=true
    bOverrideAngularDamping=true
    bUsePlanarLimit=false
    simulationSpace=RootRelative
    alphaBoolBlend {blendOption=Linear}
    boundBone {boneName=dyn_zipper}
    componentAppliedLinearAccClamp=<3500 3500 3500>
    componentLinearAccScale=<0.9 0.9 0.9>
    componentLinearVelScale=<0.9 0.9 0.9>
    physicsBodyDefinitions [{
      sphereCollisionRadius=4
      boundBone {boneName=dyn_zipper}
      boxExtents=<25 15 15>
      constraintSetup {angularLimitsMax=<1 15 5>, angularLimitsMin=<-1 -15 -35>, angularTarget=<1 0 0>}
      localJointOffset=<2 0 0>
     }]
    relativeSpaceBone {boneName=pelvis}
    sphericalLimits [{limitRadius=6, drivingBone {boneName=upperarm_twist_02_r}, sphereLocalOffset=<0 0 3>}]
   }
   Output Pose(Result=Component To Local(ComponentPose=@n57)) {layerGroup=None}
   n58 = Control Rig(Source=Copy Pose From Mesh() {bCopyCurves=true, bUseAttachedParent=true}) {
    alphaBoolBlend {blendOption=Linear}
    controlRigAssetReference {blueprintRigClass=&/F_MED_DeformRig_Veiled_ControlRig.F_MED_DeformRig_Veiled_ControlRig_C}
   }
   n59 = "Constraint - dyn_back_strap_r_MSTR (PPPPPPP)"!(
    ComponentPose=Local To Component(LocalPose=@n58)
    Alpha=1
    ConstraintWeights_0=0.0187
    ConstraintWeights_1:float=0.0076
    ConstraintWeights_2:float=0.027
    ConstraintWeights_3:float=0.0058
    ConstraintWeights_4:float=0.6044
    ConstraintWeights_5:float=0.1273
    ConstraintWeights_6:float=0.2091
   ) -> (Pose) {
    alphaBoolBlend {blendOption=Linear}
    boneToModify {boneName=dyn_back_strap_r_MSTR}
    constraintSetup [
     {targetBone {boneName=spine_04}}
     {targetBone {boneName=spine_03}}
     {targetBone {boneName=spine_02}}
     {targetBone {boneName=spine_01}}
     {targetBone {boneName=pelvis}}
     {targetBone {boneName=thigh_twist_01_r}}
     {targetBone {boneName=deform_glute_r}}
    ]
    constraintWeights [1,1,1,1,1,1,1]
   }
   n60 = "Constraint - dyn_back_strap_l_MSTR (PPPPPP)"!(
    ComponentPose=@n59
    Alpha=1
    ConstraintWeights_0=0.0036
    ConstraintWeights_1:float=0.0066
    ConstraintWeights_2:float=0.5666
    ConstraintWeights_3:float=0.18
    ConstraintWeights_4:float=0.0026
    ConstraintWeights_5:float=0.2406
   ) -> (Pose) {
    alphaBoolBlend {blendOption=Linear}
    boneToModify {boneName=dyn_back_strap_l_MSTR}
    constraintSetup [
     {targetBone {boneName=spine_04}}
     {targetBone {boneName=spine_02}}
     {targetBone {boneName=pelvis}}
     {targetBone {boneName=thigh_twist_01_l}}
     {targetBone {boneName=thigh_l}}
     {targetBone {boneName=deform_glute_l}}
    ]
    constraintWeights [1,1,1,1,1,1]
   }
   n61 = "Constraint - dyn_arm_loop_bk_MSTR (PPPPPPPPP)"!(
    ComponentPose=@n60
    Alpha=1
    ConstraintWeights_0=0
    ConstraintWeights_1:float=0.6477
    ConstraintWeights_2:float=0.0903
    ConstraintWeights_3:float=0.2529
    ConstraintWeights_4:float=0.0006
    ConstraintWeights_5:float=0.0008
    ConstraintWeights_6:float=0.0049
    ConstraintWeights_7:float=0
    ConstraintWeights_8:float=0.0027
   ) -> (Pose) {
    alphaBoolBlend {blendOption=Linear}
    boneToModify {boneName=dyn_arm_loop_bk_MSTR}
    constraintSetup [
     {targetBone {boneName=spine_04}}
     {targetBone {boneName=upperarm_twist_01_r}}
     {targetBone {boneName=upperarm_r}}
     {targetBone {boneName=upperarm_twist_02_r}}
     {targetBone {boneName=lowerarm_r}}
     {targetBone {boneName=lowerarm_twist_02_r}}
     {targetBone {boneName=deform_elbow_out_r}}
     {targetBone {boneName=deform_elbow_in_r}}
     {targetBone {boneName=deform_upperarm_bk_r}}
    ]
    constraintWeights [1,1,1,1,1,1,1,1,1]
   }
   n62 = "Constraint - dyn_bag_MSTR (PPPPPPPPP)"!(
    ComponentPose=@n61
    Alpha=1
    ConstraintWeights_0=0.0279
    ConstraintWeights_1:float=0.0077
    ConstraintWeights_2:float=0.1299
    ConstraintWeights_3:float=0.5873
    ConstraintWeights_4:float=0.0641
    ConstraintWeights_5:float=0.0351
    ConstraintWeights_6:float=0.107
    ConstraintWeights_7:float=0.0376
    ConstraintWeights_8:float=0.0035
   ) -> (Pose) {
    alphaBoolBlend {blendOption=Linear}
    boneToModify {boneName=dyn_bag_MSTR}
    constraintSetup [
     {targetBone {boneName=spine_04}}
     {targetBone {boneName=spine_03}}
     {targetBone {boneName=spine_02}}
     {targetBone {boneName=pelvis}}
     {targetBone {boneName=thigh_twist_01_l}}
     {targetBone {boneName=thigh_twist_01_r}}
     {targetBone {boneName=deform_groin_l}}
     {targetBone {boneName=deform_groin_r}}
     {targetBone {boneName=dyn_sweater_frt_r}}
    ]
    constraintWeights [1,1,1,1,1,1,1,1,1]
   }
   n63 = Control Rig(Source=Component To Local(ComponentPose=@n62)) {
    alphaBoolBlend {blendOption=Linear}
    controlRigAssetReference {blueprintRigClass=&/F_MED_Veiled_CtrlRig.F_MED_Veiled_CtrlRig_C}
   }
   n64 = "Constraint - dyn_zipper_MSTR (PPPPPPPP)"!(
    ComponentPose=Local To Component(LocalPose=@n63)
    Alpha=1
    ConstraintWeights_0=0.0059
    ConstraintWeights_1:float=0.0493
    ConstraintWeights_2:float=0.3036
    ConstraintWeights_3:float=0.0148
    ConstraintWeights_4:float=0.0364
    ConstraintWeights_5:float=0.0162
    ConstraintWeights_6:float=0.0914
    ConstraintWeights_7:float=0.4824
   ) -> (Pose) {
    alphaBoolBlend {blendOption=Linear}
    boneToModify {boneName=dyn_zipper_MSTR}
    constraintSetup [
     {targetBone {boneName=spine_04}}
     {targetBone {boneName=spine_02}}
     {targetBone {boneName=pelvis}}
     {targetBone {boneName=thigh_twist_01_l}}
     {targetBone {boneName=thigh_twist_01_r}}
     {targetBone {boneName=deform_groin_l}}
     {targetBone {boneName=deform_groin_r}}
     {targetBone {boneName=dyn_bag}}
    ]
    constraintWeights [1,1,1,1,1,1,1,1]
   }
   n65 = "Transform (Modify) Bone - Bone: floorCOL"!(ComponentPose=@n64, bAlphaBoolEnabled:bool=Get FloorCOL() -> (FloorCOL:bool), Translation="0,0,-100", Rotation="0,0,0", Scale="1,1,1") -> (Pose) {
    alphaInputType=Bool
    rotationMode=BMM_Additive
    rotationSpace=BCS_BoneSpace
    translationMode=BMM_Additive
    translationSpace=BCS_BoneSpace
    alphaBoolBlend {blendInTime=0.1, blendOption=Linear, blendOutTime=0.5}
    alphaScaleBiasClamp {bInterpResult=true}
    boneToModify {boneName=floorCOL}
    translation {z=-100}
   }
   n66 = "Transform (Modify) Bone - Bone: dyn_arm_loop_bk_MSTR"!(ComponentPose=@n65, bAlphaBoolEnabled:bool=Get bIsMoving2D(), Translation="0,0,0", Rotation="0,-45,0", Scale="1,1,1") -> (Pose) {
    alphaInputType=Bool
    rotationMode=BMM_Additive
    rotationSpace=BCS_BoneSpace
    translationMode=BMM_Additive
    translationSpace=BCS_BoneSpace
    alphaBoolBlend {blendInTime=0.1, blendOption=Linear, blendOutTime=0.5}
    boneToModify {boneName=dyn_arm_loop_bk_MSTR}
    rotation {yaw=-45}
   }
   n67 = "Transform (Modify) Bone - Bone: dyn_loop_r_bk_MSTR"!(ComponentPose=@n66, bAlphaBoolEnabled:bool=Get bIsCrouching(), Translation="0,0,0", Rotation="-15,-45,0", Scale="1,1,1") -> (Pose) {
    alphaInputType=Bool
    rotationMode=BMM_Additive
    rotationSpace=BCS_BoneSpace
    translationMode=BMM_Additive
    translationSpace=BCS_BoneSpace
    alphaBoolBlend {blendInTime=0.1, blendOption=Linear, blendOutTime=0.1}
    boneToModify {boneName=dyn_loop_r_bk_MSTR}
    rotation {pitch=-15, yaw=-45}
   }
   n68 = Get bIsSliding()
   n69 = "Transform (Modify) Bone - Bone: dyn_back_strap_l_MSTR"!(ComponentPose=@n67, bAlphaBoolEnabled:bool=@n68, Translation="0,0,0", Rotation="25,-60,0.026", Scale="1,1,1") -> (Pose) {
    alphaInputType=Bool
    rotationMode=BMM_Additive
    rotationSpace=BCS_BoneSpace
    translationMode=BMM_Additive
    translationSpace=BCS_BoneSpace
    alphaBoolBlend {blendInTime=0.1, blendOption=Linear, blendOutTime=0.1}
    boneToModify {boneName=dyn_back_strap_l_MSTR}
    rotation=<rot 25 -60 0.026>
   }
   n70 = "Transform (Modify) Bone - Bone: dyn_back_strap_r_MSTR"!(ComponentPose=@n69, bAlphaBoolEnabled:bool=@n68, Translation="0,0,0", Rotation="0,-60,0", Scale="1,1,1") -> (Pose) {
    alphaInputType=Bool
    rotationMode=BMM_Additive
    rotationSpace=BCS_BoneSpace
    translationMode=BMM_Additive
    translationSpace=BCS_BoneSpace
    alphaBoolBlend {blendInTime=0.1, blendOption=Linear, blendOutTime=0.1}
    boneToModify {boneName=dyn_back_strap_r_MSTR}
    rotation {yaw=-60}
   }
   n71 = "Transform (Modify) Bone - Bone: dyn_arm_loop_01"(ComponentPose=@n70, Alpha=Get elbow R rot(), Translation="0,0,0", Rotation="45,0,0", Scale="1,1,1") {
    rotationMode=BMM_Additive
    rotationSpace=BCS_BoneSpace
    translationMode=BMM_Additive
    translationSpace=BCS_BoneSpace
    alphaBoolBlend {blendOption=Linear}
    boneToModify {boneName=dyn_arm_loop_01}
    rotation {pitch=45}
   }
   n72 = Get Arm R rot() -> (Arm R rot:double)
   n73 = "Transform (Modify) Bone - Bone: dyn_arm_loop_01"(ComponentPose=@n71, Alpha=@n72, Translation="0,0,0", Rotation="15,15,0", Scale="1,1,1") {
    rotationMode=BMM_Additive
    rotationSpace=BCS_BoneSpace
    translationMode=BMM_Additive
    translationSpace=BCS_BoneSpace
    alphaBoolBlend {blendOption=Linear}
    boneToModify {boneName=dyn_arm_loop_01}
    rotation {pitch=15, yaw=15}
   }
   n74 = "Transform (Modify) Bone - Bone: dyn_arm_loop_bk_MSTR"!(ComponentPose=@n73, Alpha:float=@n72, Translation="1,-3,-2", Rotation="45,-35,0", Scale="1,1,1") -> (Pose) {
    rotationMode=BMM_Additive
    rotationSpace=BCS_BoneSpace
    translationMode=BMM_Additive
    translationSpace=BCS_BoneSpace
    alphaBoolBlend {blendOption=Linear}
    boneToModify {boneName=dyn_arm_loop_bk_MSTR}
    rotation {pitch=45, yaw=-35}
    translation=<1 -3 -2>
   }
   Save cached pose 'presim'(Pose=Component To Local(ComponentPose=@n74)) {cacheName=presim}
  ]}
 EventGraph {nodes [
   Event BlueprintUpdateAnimation()
   > n1 = Sequence!(execute) -> (then_0, then_1, then_2:Exec, then_3:Exec, then_4:Exec, then_5:Exec)
   > n2 = Gravity Override State Function DEPRECATED(
    CharacterPartAnimInstance=Self-Reference()
    GameTimeInSeconds=GetGameTimeInSeconds()
    GravityOverrideParams=Get GravityOverrideParams() -> (GravityOverrideParams:Gravity Override Params Struct Structure)
   )
   > Set GravityOverride(GravityOverride=@n2)
   n3 = Get skeletal_mesh() -> (skeletal_mesh:Skeletal Mesh Component Object Reference)
   > Set elbow R rot(elbow R rot:double=GetAlphaFromReferenceTransformDelta(SkeletalMeshComponent=@n3, BoneName=lowerarm_r, BaseName=upperarm_r, InputRange=(X=0,Y=90), TwistAxis="1,0,0")) -> (Output_Get:double)
   > Set Arm R rot(Arm R rot:double=GetAlphaFromReferenceTransformDelta(SkeletalMeshComponent=@n3, BoneName=upperarm_r, BaseName=spine_05, InputRange=(X=40,Y=100), TwistAxis="1,0,0")) -> (Output_Get:double)
   n4 = CheckForBackbling()
   > n5 = Branch(Condition=@n4)
   > Set backpack(backpack:bool=true) -> (Output_Get:bool)
   > n6 = Branch(Condition=@n4.bIsBackblingACape)
   > Set Cape(Cape:bool=true) -> (Output_Get:bool)
   Set Cape(execute=@n6.else, Cape=false)
   Set backpack(execute=@n5.else, backpack=false)
   n7 = Get bIsMoving2D()
   n8 = Branch(execute=@n1.then_2, Condition=AND Boolean(A=Get bIsCrouching(), B=NOT Boolean(A=@n7)))
   > Set still_crouch(still_crouch:bool=true) -> (Output_Get:bool)
   Set still_crouch(execute=@n8.else, still_crouch=false)
   n9 = Branch(execute=@n1.then_3, Condition=OR Boolean!(A=Get bIsJumping(), B=Get bIsFalling(), C:bool=Get bIsDBNO()) -> (ReturnValue))
   > Set FloorCOL(FloorCOL:bool=true) -> (Output_Get:bool)
   Set FloorCOL(execute=@n9.else, FloorCOL=false)
   n10 = Get skeletal_mesh()
   Set ThighRotL(
    execute=@n1.then_4
    ThighRotL:double=GetAlphaFromReferenceTransformDelta(SkeletalMeshComponent=@n10, BoneName=thigh_l, BaseName=pelvis, InputRange=(X=-20,Y=-100), TwistAxis="1,0,0")
   ) -> (Output_Get:double)
   > Set ThighRotR(ThighRotR:double=GetAlphaFromReferenceTransformDelta(SkeletalMeshComponent=@n10, BoneName=thigh_r, BaseName=pelvis, InputRange=(X=-20,Y=-100), TwistAxis="1,0,0")) -> (Output_Get:double)
   > n11 = PlayerMovement(AnimInstance=Self-Reference(), XDotProductCoefficient=1, YDotProductCoefficient=1, ZDotProductCoefficient=1)
   > Set BwdMovement(BwdMovement:float=MapRangeClamped(Value=@n11.YMovement, InRangeA=0, InRangeB=-1, OutRangeA=1, OutRangeB=0.4)) -> (Output_Get:float)
   > n12 = Branch(Condition=Get bIsSkyDiving())
   > n13 = Branch(Condition="float > float"(A=Get PawnVelocityZ(), B=50) {is=title K2Node_CallFunction})
   > Set BwdMovement(BwdMovement=float * float(A=Get BwdMovement(), B=0))
   Set BwdMovement(execute=[@n12.else, @n13.else], BwdMovement=float * float(A=Get BwdMovement(), B=1))
   n14 = Branch(execute=@n1.then_5, Condition="integer >= integer"(A=2, B=Get CurrentLOD()))
   > Set LOD2(LOD2:bool=true) -> (Output_Get:bool)
   Set LOD2(execute=@n14.else, LOD2=false)
   Event BlueprintInitializeAnimation()
   > Set skeletal_mesh(skeletal_mesh:Skeletal Mesh Component Object Reference=GetOwnerSkeletalMeshComponent()) -> (Output_Get:Skeletal Mesh Component Object Reference)
   n15 = TryGetPawnOwner()
   > Is Valid(InputObject=@n15)
   > n16 = Cast To FortPlayerPawnAthena(Object=@n15)
   > Set playerPawn(playerPawn:Fort Player Pawn Athena Object Reference=@n16) -> (Output_Get:Fort Player Pawn Athena Object Reference)
   n17 = Event OnEndFalling()
   Set jumping(execute=[@n17.then, @n18.then], jumping:double=0) -> (Output_Get:double)
   n18 = Event OnEndJumping()
   n19 = Event OnBeginFalling()
   Set jumping(execute=[@n19.then, @n20.then], jumping=1)
   n20 = Event OnBeginJumping()
  ]}
]
variables [
 Arm R rot {type=number}
 backpack {type=boolean}
 BwdMovement {type=number}
 Cape {type=boolean}
 elbow R rot {type=number}
 FloorCOL {type=boolean}
 GravityOverride {type=Vector}
 GravityOverrideParams {type=GravityOverrideParamsStruct}
 jumping {type=number}
 LOD2 {type=boolean}
 playerPawn {type=FortPlayerPawnAthena}
 skeletal_mesh {type=SkeletalMeshComponent}
 still_crouch {type=boolean}
 ThighRotL {type=number}
 ThighRotR {type=number}
]
