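# Physics asset for the medium female Veiled player body, providing capsule colliders that, by its name, serve the cape.
# Capsules cover pelvis, spine, clavicles, head, feet, a hood bone and two back-strap bone chains, with second crosswise capsules widening spine_02 to spine_04.
# Arm, leg and neck bodies carry no shapes, apart from the feet.
# Constraints link bodies along the skeleton with free swing and twist, the straps hanging from pelvis.
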PhysicsAsset /Game/Characters/Player/Female/Medium/Bodies/F_MED_Veiled/Meshes/F_MED_Veiled_CapeColliders
bodies {
 calf_l []
 calf_r []
 clavicle_l [{shapeType=Capsule, length=7.6964, radius=6.3853, center=<4.0147 -0.8354 -1.7419>, rotation=<rot -77.5566 177.3915 7.7314>}]
 clavicle_r [{shapeType=Capsule, length=7.6964, radius=8.2957, center=<-4.0147 0.8354 1.7419>, rotation=<rot -77.5565 177.3915 -172.2685>}]
 dyn_back_strap_l_01 [{shapeType=Capsule, length=8, radius=6, shapeName=dyn_back_strap_l_01_capsule, center=<2 8 0>, rotation=<rot 90 0 0>}]
 dyn_back_strap_l_02 [{shapeType=Capsule, length=8, radius=6, shapeName=dyn_back_strap_l_02_capsule, center=<9 5 0>, rotation=<rot 90 0 0>}]
 dyn_back_strap_l_03 [{shapeType=Capsule, length=8, radius=6, shapeName=dyn_back_strap_l_03_capsule, center=<9.791 5 0>, rotation=<rot 90 0 0>}]
 dyn_back_strap_l_04 [{shapeType=Capsule, length=13.5, radius=6, shapeName=dyn_back_strap_l_04_capsule, center=<10.4255 5 0>, rotation=<rot 90 0 0>}]
 dyn_back_strap_l_MSTR [{shapeType=Capsule, length=0.505, radius=0.505, shapeName=dyn_back_strap_l_MSTR_capsule, rotation=<rot 0 -90 0>}]
 dyn_back_strap_r_01 [{shapeType=Capsule, length=8, radius=6, shapeName=dyn_back_strap_r_01_capsule, center=<2 8 0>, rotation=<rot 90 0 0>}]
 dyn_back_strap_r_02 [{shapeType=Capsule, length=8, radius=6, shapeName=dyn_back_strap_r_02_capsule, center=<9 5 0>, rotation=<rot 90 0 0>}]
 dyn_back_strap_r_03 [{shapeType=Capsule, length=8, radius=6, shapeName=dyn_back_strap_r_03_capsule, center=<9 5 0>, rotation=<rot 90 0 0>}]
 dyn_back_strap_r_04 [{shapeType=Capsule, length=8, radius=6, shapeName=dyn_back_strap_r_04_capsule, center=<9 5 0>, rotation=<rot 90 0 0>}]
 dyn_back_strap_r_05 [{shapeType=Capsule, length=13.75, radius=6, shapeName=dyn_back_strap_r_05_capsule, center=<13.5435 5 0>, rotation=<rot 90 0 0>}]
 dyn_back_strap_r_MSTR [{shapeType=Capsule, length=0.505, radius=0.505, shapeName=dyn_back_strap_r_MSTR_capsule, rotation=<rot 0 -90 0>}]
 dyn_hood_02 [{
   shapeType=Capsule
   length=9.0674
   radius=6.1996
   shapeName=dyn_hood_02_capsule
   center=<0.8723 0.9011 -2.3837>
   rotation=<rot 0 -171.1056 -89.5516>
  }]
 foot_l [{shapeType=Capsule, length=12.5982, radius=6.5127, center=<-2.2072 6.0534 -0.3367>, rotation=<rot 0.0388 1.4701 -89.3982>}]
 foot_r [{shapeType=Capsule, length=12.5982, radius=6.5127, center=<2.2072 -6.0534 0.3367>, rotation=<rot 0.0388 1.4701 90.6018>}]
 head [{shapeType=Capsule, length=0, radius=18, center=<9.0697 0.0746 0>, rotation=<rot 0 -90 0>}]
 lowerarm_l []
 lowerarm_r []
 neck_02 []
 pelvis [{shapeType=Capsule, length=13.6528, radius=12.4096, center=<-3.5964 -1.7256 -0.1981>}]
 spine_01 [{shapeType=Capsule, length=13.2671, radius=10.4854, center=<-2.8183 -1.0873 0>}]
 spine_02 [
  {shapeType=Capsule, length=8.2984, radius=10.7354, center=<1.179 0.3339 0>}
  {shapeType=Capsule, length=17.9375, radius=12.0625, center=<-1.8317 2.606 0>, rotation=<rot 90 0 0>}
 ]
 spine_03 [
  {shapeType=Capsule, length=9.7984, radius=9.9854, center=<-0.0747 0.7337 0>}
  {shapeType=Capsule, length=20.7188, radius=9.2812, center=<-1.2326 2.6059 0>, rotation=<rot 90 0 0>}
 ]
 spine_04 [
  {shapeType=Capsule, length=13.0796, radius=10.3916, center=<5.5182 1.4112 0>}
  {shapeType=Capsule, length=10.0312, radius=9.2812, center=<-1.2326 1.308 0>, rotation=<rot 90 0 0>}
 ]
 spine_05 [{shapeType=Capsule, length=10.9859, radius=10.0791, center=<5.6377 0.3227 0>}]
 thigh_l []
 thigh_r []
 upperarm_l []
 upperarm_r []
}
constraints [
 {
  bone1Name=spine_01
  bone2Name=pelvis
  swing1LimitDegrees=45
  swing1Motion=Free
  swing2LimitDegrees=45
  swing2Motion=Free
  twistLimitDegrees=45
  twistMotion=Free
 }
 {
  bone1Name=spine_02
  bone2Name=spine_01
  swing1LimitDegrees=45
  swing1Motion=Free
  swing2LimitDegrees=45
  swing2Motion=Free
  twistLimitDegrees=45
  twistMotion=Free
 }
 {
  bone1Name=spine_03
  bone2Name=spine_02
  swing1LimitDegrees=45
  swing1Motion=Free
  swing2LimitDegrees=45
  swing2Motion=Free
  twistLimitDegrees=45
  twistMotion=Free
 }
 {
  bone1Name=spine_04
  bone2Name=spine_03
  swing1LimitDegrees=45
  swing1Motion=Free
  swing2LimitDegrees=45
  swing2Motion=Free
  twistLimitDegrees=45
  twistMotion=Free
 }
 {
  bone1Name=spine_05
  bone2Name=spine_04
  swing1LimitDegrees=45
  swing1Motion=Free
  swing2LimitDegrees=45
  swing2Motion=Free
  twistLimitDegrees=45
  twistMotion=Free
 }
 {
  bone1Name=clavicle_l
  bone2Name=spine_05
  swing1LimitDegrees=45
  swing1Motion=Free
  swing2LimitDegrees=45
  swing2Motion=Free
  twistLimitDegrees=45
  twistMotion=Free
 }
 {
  bone1Name=upperarm_l
  bone2Name=clavicle_l
  swing1LimitDegrees=45
  swing1Motion=Free
  swing2LimitDegrees=45
  swing2Motion=Free
  twistLimitDegrees=45
  twistMotion=Free
 }
 {
  bone1Name=lowerarm_l
  bone2Name=upperarm_l
  swing1LimitDegrees=45
  swing1Motion=Free
  swing2LimitDegrees=45
  swing2Motion=Free
  twistLimitDegrees=45
  twistMotion=Free
 }
 {
  bone1Name=neck_02
  bone2Name=spine_05
  swing1LimitDegrees=45
  swing1Motion=Free
  swing2LimitDegrees=45
  swing2Motion=Free
  twistLimitDegrees=45
  twistMotion=Free
 }
 {
  bone1Name=head
  bone2Name=neck_02
  swing1LimitDegrees=45
  swing1Motion=Free
  swing2LimitDegrees=45
  swing2Motion=Free
  twistLimitDegrees=45
  twistMotion=Free
 }
 {
  bone1Name=calf_l
  bone2Name=thigh_l
  swing1LimitDegrees=45
  swing1Motion=Free
  swing2LimitDegrees=45
  swing2Motion=Free
  twistLimitDegrees=45
  twistMotion=Free
 }
 {
  bone1Name=thigh_l
  bone2Name=pelvis
  swing1LimitDegrees=45
  swing1Motion=Free
  swing2LimitDegrees=45
  swing2Motion=Free
  twistLimitDegrees=45
  twistMotion=Free
 }
 {
  bone1Name=foot_l
  bone2Name=calf_l
  swing1LimitDegrees=45
  swing1Motion=Free
  swing2LimitDegrees=45
  swing2Motion=Free
  twistLimitDegrees=45
  twistMotion=Free
 }
 {
  bone1Name=clavicle_r
  bone2Name=spine_05
  swing1LimitDegrees=45
  swing1Motion=Free
  swing2LimitDegrees=45
  swing2Motion=Free
  twistLimitDegrees=45
  twistMotion=Free
 }
 {
  bone1Name=lowerarm_r
  bone2Name=upperarm_r
  swing1LimitDegrees=45
  swing1Motion=Free
  swing2LimitDegrees=45
  swing2Motion=Free
  twistLimitDegrees=45
  twistMotion=Free
 }
 {
  bone1Name=thigh_r
  bone2Name=pelvis
  swing1LimitDegrees=45
  swing1Motion=Free
  swing2LimitDegrees=45
  swing2Motion=Free
  twistLimitDegrees=45
  twistMotion=Free
 }
 {
  bone1Name=calf_r
  bone2Name=thigh_r
  swing1LimitDegrees=45
  swing1Motion=Free
  swing2LimitDegrees=45
  swing2Motion=Free
  twistLimitDegrees=45
  twistMotion=Free
 }
 {
  bone1Name=foot_r
  bone2Name=calf_r
  swing1LimitDegrees=45
  swing1Motion=Free
  swing2LimitDegrees=45
  swing2Motion=Free
  twistLimitDegrees=45
  twistMotion=Free
 }
 {
  bone1Name=upperarm_r
  bone2Name=clavicle_r
  swing1LimitDegrees=45
  swing1Motion=Free
  swing2LimitDegrees=45
  swing2Motion=Free
  twistLimitDegrees=45
  twistMotion=Free
 }
 {
  bone1Name=dyn_back_strap_r_05
  bone2Name=dyn_back_strap_r_04
  swing1LimitDegrees=45
  swing1Motion=Free
  swing2LimitDegrees=45
  swing2Motion=Free
  twistLimitDegrees=45
  twistMotion=Free
 }
 {
  bone1Name=dyn_back_strap_r_01
  bone2Name=dyn_back_strap_r_MSTR
  swing1LimitDegrees=45
  swing1Motion=Free
  swing2LimitDegrees=45
  swing2Motion=Free
  twistLimitDegrees=45
  twistMotion=Free
 }
 {
  bone1Name=dyn_back_strap_r_02
  bone2Name=dyn_back_strap_r_01
  swing1LimitDegrees=45
  swing1Motion=Free
  swing2LimitDegrees=45
  swing2Motion=Free
  twistLimitDegrees=45
  twistMotion=Free
 }
 {
  bone1Name=dyn_back_strap_r_03
  bone2Name=dyn_back_strap_r_02
  swing1LimitDegrees=45
  swing1Motion=Free
  swing2LimitDegrees=45
  swing2Motion=Free
  twistLimitDegrees=45
  twistMotion=Free
 }
 {
  bone1Name=dyn_back_strap_r_04
  bone2Name=dyn_back_strap_r_03
  swing1LimitDegrees=45
  swing1Motion=Free
  swing2LimitDegrees=45
  swing2Motion=Free
  twistLimitDegrees=45
  twistMotion=Free
 }
 {
  bone1Name=dyn_back_strap_l_04
  bone2Name=dyn_back_strap_l_03
  swing1LimitDegrees=45
  swing1Motion=Free
  swing2LimitDegrees=45
  swing2Motion=Free
  twistLimitDegrees=45
  twistMotion=Free
 }
 {
  bone1Name=dyn_back_strap_l_01
  bone2Name=dyn_back_strap_l_MSTR
  swing1LimitDegrees=45
  swing1Motion=Free
  swing2LimitDegrees=45
  swing2Motion=Free
  twistLimitDegrees=45
  twistMotion=Free
 }
 {
  bone1Name=dyn_back_strap_l_02
  bone2Name=dyn_back_strap_l_01
  swing1LimitDegrees=45
  swing1Motion=Free
  swing2LimitDegrees=45
  swing2Motion=Free
  twistLimitDegrees=45
  twistMotion=Free
 }
 {
  bone1Name=dyn_back_strap_l_03
  bone2Name=dyn_back_strap_l_02
  swing1LimitDegrees=45
  swing1Motion=Free
  swing2LimitDegrees=45
  swing2Motion=Free
  twistLimitDegrees=45
  twistMotion=Free
 }
 {
  bone1Name=dyn_back_strap_l_MSTR
  bone2Name=pelvis
  swing1LimitDegrees=45
  swing1Motion=Free
  swing2LimitDegrees=45
  swing2Motion=Free
  twistLimitDegrees=45
  twistMotion=Free
 }
 {
  bone1Name=dyn_back_strap_r_MSTR
  bone2Name=pelvis
  swing1LimitDegrees=45
  swing1Motion=Free
  swing2LimitDegrees=45
  swing2Motion=Free
  twistLimitDegrees=45
  twistMotion=Free
 }
 {
  bone1Name=dyn_hood_02
  bone2Name=spine_05
  swing1LimitDegrees=45
  swing1Motion=Free
  swing2LimitDegrees=45
  swing2Motion=Free
  twistLimitDegrees=45
  twistMotion=Free
 }
]
